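# Modular Control Rig module for a Sidekick back leg, building an IK rig with a hock joint.
# Construction resolves connectors, then spawns hip or scapula, IK foot, heel and toe pivot, roll, ball/toe FK controls.
# It also adds a computed knee pole vector, knee rotator and ankle roll control.
# Initial chain lengths and hock-aim rotations are cached, including a squash target extrapolated from a maximum hock angle.
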ControlRigBlueprint /EpicControlRig/Modules/Sidekick/Back_Leg/CRM_Epic_Sidekick_BackLeg_v01
graphs [
 CollapseNode_ContainedGraph {nodes [
   RigVMFunctionEntryNode()
   > RigVMFunction_Sequence = RigVMAggregateNode()
   n1 = &RigVMFunction_Sequence(ExecuteContext=@RigVMFunction_Sequence.A)
   n2 = RigVMVariableNode() {variable=Ball Connector}
   n3 = RigVMVariableNode() {variable=Toe Connector}
   RigVMVariableNode(
    Value=&RigVMFunction_MathVectorDistance(
     A=&RigUnit_GetTransform(bInitial=true, Item=@n2.Value).Transform.Translation
     B=&RigUnit_GetTransform(bInitial=true, Item=@n3.Value).Transform.Translation
    ).Result
    ExecuteContext=@n1.A
   ) {variable=Initial Ball Length}
   n4 = RigVMVariableNode() {variable=Hip Control}
   n5 = RigVMVariableNode() {variable=Knee Connector}
   > RigVMVariableNode(Value=&RigVMFunction_MathVectorDistance(
     A=&RigUnit_GetTransform(bInitial=true, Item=@n4.Value).Transform.Translation
     B=&RigUnit_GetTransform(bInitial=true, Item=@n5.Value).Transform.Translation
    ).Result) {variable=Initial Thigh Length}
   Make_Array = RigVMDispatchNode(
    Values=((Type=Bone,Name="torso_1"),((),Type=Bone,Name="torso_2"),((),Type=Bone,Name="torso_3"),((),Type=Bone,Name="torso_4"),((Type=Bone,Name="torso_1"),(Type=Bone,Name="torso_2"),(Type=Bone,Name="torso_3"),(Type=Bone,Name="torso_4"),Type=None,Name="None"))
    Values.0=@n4.Value
    Values.1=@n5.Value
    Values.2=RigVMVariableNode() {variable=Ankle Connector}.Value
    Values.4=@n3.Value
    Values.3=@n2.Value
   ) {template="DISPATCH_RigVMDispatch_ArrayMake(in Values,out Array)"}
   Get Chain Length = RigVMFunctionReferenceNode(Initial=true, Items=@Make_Array.Array, ExecuteContext=@n1.B)
   > RigVMVariableNode(Value=@"Get Chain Length".Length) {variable=Chain Length}
   > RigVMVariableNode(Value=&RigVMFunction_MathDoubleSub(A=@"Get Chain Length".Length, B=RigVMVariableNode() {variable=Initial Ball Length}.Value).Result) {variable=Hock Chain Length}
   n6 = &RigUnit_GetTransform(bInitial=true, Item=RigVMVariableNode() {variable=Hip Control}.Value)
   n7 = &RigVMFunction_MathVectorDistance(
    B=@n6.Transform.Translation
    A=&RigUnit_GetTransform(bInitial=true, Item=RigVMVariableNode() {variable=Ankle Connector}.Value).Transform.Translation
   )
   > RigVMVariableNode(Value=@n7.Result) {variable=Initial Thigh to Ankle Length}
   n8 = RigVMVariableNode() {variable=Initial Thigh Length}
   Add = &RigVMFunction_MathDoubleAdd(B=@n7.Result, A=@n8.Value)
   Divide = &RigVMFunction_MathDoubleDiv(B=@Add.Result, A=@n8.Value)
   > RigVMVariableNode(Value=@Divide.Result) {variable=PV Null Ratio}
   Subtract = /Script/RigVM.RigVMFunction_MathVectorSub(
    B=&RigUnit_GetTransform(bInitial=true, Item=RigVMVariableNode() {variable=IK Foot Control}.Value).Transform.Translation
    A=@n6.Transform.Translation
   ) {is=struct RigVMUnitNode, template="Subtract::Execute(in A,in B,out Result)"}
   RigVMFunction_MathVectorLength = /Script/RigVM.RigVMFunction_MathVectorLength(Value=@Subtract.Result) {is=struct RigVMUnitNode, template="Length::Execute(in Value,out Result)"}
   > RigVMVariableNode(Value=@RigVMFunction_MathVectorLength.Result) {variable=Initial Distance Hip to IK Target}
   RigVMFunctionReturnNode(ExecuteContext=@RigVMFunction_Sequence.D)
  ]}
 CollapseNode_ContainedGraph {nodes [
   n1 = RigVMFunctionEntryNode()
   RigVMFunction_Sequence = &RigVMFunction_Sequence(ExecuteContext=@n1.ExecutePin)
   RigVMFunctionReturnNode(ExecutePin=@RigVMFunction_Sequence.B)
   n2 = RigVMVariableNode() {variable=Ball Connector}
   Solve Hock Aim Null = RigVMFunctionReferenceNode(
    Initial=true
    Execute=@RigVMFunction_Sequence.A
    Base=RigVMVariableNode() {variable=Hip Connector}.Value
    Ankle=RigVMVariableNode() {variable=Ankle Connector}.Value
    End=@n2.Value
   )
   RigVMVariableNode(ExecuteContext=@"Solve Hock Aim Null".Execute, Value=@"Solve Hock Aim Null".Result) {variable=Hock Aim Null Transform}
   > n3 = &RigUnit_HierarchyAddNull(
    Name=Ankle Roll
    Space=GlobalSpace
    Transform=(Rotation=(X=0,Y=0,Z=0,W=1),Translation=(X=0,Y=0,Z=0),Scale3D=(X=1,Y=1,Z=1))
    Transform.Translation=&RigUnit_GetTransform(bInitial=true, Item=@n2.Value).Transform.Translation
    Parent=RigVMVariableNode() {variable=Parent}.Value
    Transform.Rotation=@"Solve Hock Aim Null".Result.Rotation
   )
   > RigVMVariableNode(Value=@n3.Item) {variable=Ankle Roll Null}
   Prepare Control Shape_0 = RigVMFunctionReferenceNode(
    Scale Premultiplier=1
    Control Shape=RigVMVariableNode() {variable=Ankle Roll Shape Settings}.Value
    Scale Factor=RigVMVariableNode() {variable=Control Scale}.Value
   )
   > SpawnControl = /Script/ControlRig.RigUnit_HierarchyAddControlTransform(
    InitialValue=(Rotation=(X=0,Y=0,Z=0,W=1),Translation=(X=0,Y=0,Z=0),Scale3D=(X=1,Y=1,Z=1))
    Name=Ankle Roll
    OffsetSpace=LocalSpace
    OffsetTransform=(Rotation=(X=0.5,Y=0.5,Z=0.5,W=0.5),Translation=(X=0,Y=0,Z=0),Scale3D=(X=1,Y=1,Z=1))
    Settings="(InitialSpace=LocalSpace,bUsePreferredRotationOrder=False,PreferredRotationOrder=YZX,Limits=(LimitTranslationX=(bMinimum=False,bMaximum=False),LimitTranslationY=(bMinimum=False,bMaximum=False),LimitTranslationZ=(bMinimum=False,bMaximum=False),LimitPitch=(b ...(truncated)"
    Settings.Shape=@"Prepare Control Shape_0".Result Control Shape
    Parent=RigVMVariableNode() {variable=Ankle Roll Null}.Value
   ) {
    is=struct RigVMUnitNode
    template="SpawnControl::Execute(in InitialValue,in Settings,in OffsetTransform,in OffsetSpace,in Parent,in Name,out Item)"
   }
   > RigVMVariableNode(Value=@SpawnControl.Item) {variable=Ankle Roll Control}
   n4 = &RigUnit_GetTransform(bInitial=true, Item=RigVMVariableNode() {variable=Ankle Connector}.Value)
   > n5 = &RigUnit_HierarchyAddNull(
    Name=Ankle Roll Offset
    Space=GlobalSpace
    Transform=(Rotation=(X=0,Y=0,Z=0,W=1),Translation=(X=0,Y=0,Z=0),Scale3D=(X=1,Y=1,Z=1))
    Transform.Translation=RigVMVariableNode() {variable=Hock Aim Null Transform}.Value.Translation
    Parent=RigVMVariableNode() {variable=Ankle Roll Control}.Value
    Transform.Rotation=@n4.Transform.Rotation
   )
   > RigVMVariableNode(Value=@n5.Item) {variable=Ankle Roll Bottom Null}
   > n6 = &RigUnit_HierarchyAddNull(Name=Ankle Roll Driver, Space=GlobalSpace, Parent=@n5.Item, Transform=@n4.Transform)
   > RigVMVariableNode(Value=@n6.Item) {variable=Ankle Driver Null}
   RigVMVariableNode() {variable=Knee Connector}
  ]}
 CollapseNode_ContainedGraph {nodes [
   n1 = RigVMFunctionEntryNode()
   RigVMFunction_Sequence = &RigVMFunction_Sequence(ExecuteContext=@n1.ExecuteContext)
   n2 = /Script/ControlRig.RigUnit_GetCandidates() {is=struct RigVMUnitNode}
   RigVMFunction_MathBoolNot = &RigVMFunction_MathBoolNot(Value=/Script/ControlRig.RigUnit_ResolveConnector(SkipSocket=False, Connector=@n2.Connector) {is=struct RigVMUnitNode}.bIsConnected)
   RigVMFunction_MathBoolAnd = &RigVMFunction_MathBoolAnd(
    B=RigVMDispatchNode(A=@n2.Connector, B=RigVMVariableNode() {variable=Child}.Value) {template="DISPATCH_RigVMDispatch_CoreEquals(in A,in B,out Result)"}.Result
    A=@RigVMFunction_MathBoolNot.Result
   )
   RigVMFunction_ControlFlowBranch = &RigVMFunction_ControlFlowBranch(Condition=@RigVMFunction_MathBoolAnd.Result, ExecuteContext=@RigVMFunction_Sequence.A)
   n3 = /Script/ControlRig.RigUnit_CollectionChildrenArray(
    TypeToSearch=Bone
    bDefaultChildren=True
    bIncludeParent=False
    bRecursive=False
    Parent=&RigUnit_ResolveConnector(SkipSocket=False, Connector=RigVMVariableNode() {variable=Parent}.Value).Result
   ) {is=struct RigVMUnitNode}
   Num = RigVMDispatchNode(Array=@n3.Items) {template="DISPATCH_RigVMDispatch_ArrayGetNum(in Array,out Num)"}
   Greater = /Script/RigVM.RigVMFunction_MathIntGreater(B=0, A=@Num.Num) {is=struct RigVMUnitNode, template="Greater::Execute(in A,in B,out Result)"}
   n4 = &RigVMFunction_ControlFlowBranch(Condition=@Greater.Result, ExecuteContext=@RigVMFunction_ControlFlowBranch.True)
   /Script/ControlRig.RigUnit_SetDefaultMatch(
    Default=RigVMDispatchNode(Index=0, Array=@n3.Items) {template="DISPATCH_RigVMDispatch_ArrayGetAtIndex(in Array,in Index,out Element)"}.Element
    ExecutePin=@n4.True
   ) {is=struct RigVMUnitNode, template="Set Default Match::Execute(in Default)"}
   RigVMFunctionReturnNode(ExecuteContext=@RigVMFunction_Sequence.B)
  ]}
 Construct Foot_ContainedGraph {nodes [
   RigVMFunctionEntryNode()
   > RigVMFunction_Sequence = RigVMAggregateNode()
   n1 = RigVMVariableNode() {variable=Is Scapula Connected}
   RigVMFunction_ControlFlowBranch = &RigVMFunction_ControlFlowBranch(Condition=@n1.Value, ExecuteContext=@RigVMFunction_Sequence.A)
   n2 = /Script/RigVM.RigVMFunction_MathDoubleGreater(B=0, A=RigVMVariableNode() {variable=IK Primary Axis}.Value.X) {is=struct RigVMUnitNode, template="Greater::Execute(in A,in B,out Result)"}
   Prepare Control Shape_5 = RigVMFunctionReferenceNode(
    Control Shape=(bVisible=True,Name="Default",Color=(R=1,G=0,B=0,A=1),Transform=(Rotation=(X=0,Y=0,Z=0,W=1),Translation=(X=0,Y=0,Z=0),Scale3D=(X=1,Y=1,Z=1)))
    Scale Factor=1
    Scale Premultiplier=0.3
   )
   Negate = /Script/RigVM.RigVMFunction_MathVectorNegate(Value=@"Prepare Control Shape_5".Result Control Shape.Transform.Translation) {is=struct RigVMUnitNode, template="Negate::Execute(in Value,out Result)"}
   RigUnit_GetTransform = &RigUnit_GetTransform(bInitial=true, Item=RigVMVariableNode() {variable=Scapula Connector}.Value)
   SpawnControl = &RigUnit_HierarchyAddControlTransform(
    InitialValue=(Rotation=(X=0,Y=0,Z=0,W=1),Translation=(X=0,Y=0,Z=0),Scale3D=(X=1,Y=1,Z=1))
    Name=Scapula
    OffsetSpace=GlobalSpace
    OffsetTransform=(Rotation=(X=0,Y=0,Z=0,W=1),Translation=(X=0,Y=0,Z=0),Scale3D=(X=1,Y=1,Z=1))
    Settings="(InitialSpace=LocalSpace,bUsePreferredRotationOrder=False,PreferredRotationOrder=YZX,Limits=(LimitTranslationX=(bMinimum=False,bMaximum=False),LimitTranslationY=(bMinimum=False,bMaximum=False),LimitTranslationZ=(bMinimum=False,bMaximum=False),LimitPitch=(b ...(truncated)"
    Settings.Shape.Transform.Translation=RigVMDispatchNode(Condition=@n2.Result, False=@Negate.Result, True=@"Prepare Control Shape_5".Result Control Shape.Transform.Translation) {template="DISPATCH_RigVMDispatch_If(in Condition,in True,in False,out Result)"}.Result
    Settings.Shape.Color=@"Prepare Control Shape_5".Result Control Shape.Color
    Settings.Shape.Name=@"Prepare Control Shape_5".Result Control Shape.Name
    Settings.Shape.Transform.Rotation=@"Prepare Control Shape_5".Result Control Shape.Transform.Rotation
    Settings.Shape.Transform.Scale3D=@"Prepare Control Shape_5".Result Control Shape.Transform.Scale3D
    Settings.Shape.bVisible=@"Prepare Control Shape_5".Result Control Shape.bVisible
    OffsetTransform.Translation=@RigUnit_GetTransform.Transform.Translation
    Parent=RigVMVariableNode() {variable=Control Parent Connector}.Value
    ExecutePin=@RigVMFunction_ControlFlowBranch.True
   ) {template="SpawnControl::Execute(in InitialValue,in Settings,in OffsetTransform,in OffsetSpace,in Parent,in Name,out Item)"}
   > RigVMVariableNode(Value=@SpawnControl.Item) {variable=Scapula Control}
   Prepare Control Shape_1 = RigVMFunctionReferenceNode(
    Scale Premultiplier=0.85
    Control Shape=RigVMVariableNode() {variable=IK Shape Settings}.Value
    Scale Factor=RigVMVariableNode() {variable=Control Scale}.Value
   )
   n3 = &RigUnit_GetTransform(bInitial=False, Item=RigVMVariableNode() {variable=Ball Connector}.Value)
   n4 = &RigUnit_HierarchyAddControlTransform(
    InitialValue=(Rotation=(X=0,Y=0,Z=0,W=1),Translation=(X=0,Y=0,Z=0),Scale3D=(X=1,Y=1,Z=1))
    Name=IK
    OffsetSpace=GlobalSpace
    OffsetTransform=(Rotation=(X=0,Y=0,Z=0,W=1),Translation=(X=0,Y=0,Z=0),Scale3D=(X=1,Y=1,Z=1))
    Settings="(InitialSpace=LocalSpace,bUsePreferredRotationOrder=False,PreferredRotationOrder=YZX,Limits=(LimitTranslationX=(bMinimum=False,bMaximum=False),LimitTranslationY=(bMinimum=False,bMaximum=False),LimitTranslationZ=(bMinimum=False,bMaximum=False),LimitPitch=(b ...(truncated)"
    Settings.Shape=@"Prepare Control Shape_1".Result Control Shape
    OffsetTransform.Scale3D=@n3.Transform.Scale3D
    OffsetTransform.Translation=@n3.Transform.Translation
    Parent=RigVMVariableNode() {variable=Pelvis Connector}.Value
    ExecutePin=@RigVMFunction_Sequence.C
   ) {template="SpawnControl::Execute(in InitialValue,in Settings,in OffsetTransform,in OffsetSpace,in Parent,in Name,out Item)"}
   > RigVMVariableNode(Value=@n4.Item) {variable=IK Foot Control}
   n5 = &RigUnit_GetTransform(
    bInitial=true
    Item=RigVMDispatchNode(
     True=RigVMVariableNode() {variable=Toe Tip Connector}.Value
     Condition=RigVMVariableNode() {variable=Is Toe Tip Connected}.Value
     False=RigVMVariableNode() {variable=Toe Connector}.Value
    ) {template="DISPATCH_RigVMDispatch_If(in Condition,in True,in False,out Result)"}.Result
   )
   Multiply = &RigVMFunction_MathTransformMul(
    A=(Rotation=(X=0,Y=0,Z=0,W=1),Translation=(X=0,Y=0,Z=0),Scale3D=(X=1,Y=1,Z=1))
    A.Scale3D=@n5.Transform.Scale3D
    A.Translation=@n5.Transform.Translation
    B=RigVMVariableNode() {variable=Toe Roll Tip Offset Transform}.Value
   ) {template="Multiply::Execute(in A,in B,out Result)"}
   Prepare Control Shape_2 = RigVMFunctionReferenceNode(
    Scale Premultiplier=0.2
    Control Shape=RigVMVariableNode() {variable=Toe Pivot Shape Settings}.Value
    Scale Factor=RigVMVariableNode() {variable=Control Scale}.Value
   )
   > n6 = &RigUnit_HierarchyAddControlTransform(
    InitialValue=(Rotation=(X=0,Y=0,Z=0,W=1),Translation=(X=0,Y=0,Z=0),Scale3D=(X=1,Y=1,Z=1))
    Name=Toe Tip
    OffsetSpace=GlobalSpace
    Settings="(InitialSpace=LocalSpace,bUsePreferredRotationOrder=False,PreferredRotationOrder=YZX,Limits=(LimitTranslationX=(bMinimum=False,bMaximum=False),LimitTranslationY=(bMinimum=False,bMaximum=False),LimitTranslationZ=(bMinimum=False,bMaximum=False),LimitPitch=(b ...(truncated)"
    OffsetTransform=@Multiply.Result
    Settings.Shape=@"Prepare Control Shape_2".Result Control Shape
    Parent=RigVMVariableNode() {variable=IK Foot Control}.Value
   ) {template="SpawnControl::Execute(in InitialValue,in Settings,in OffsetTransform,in OffsetSpace,in Parent,in Name,out Item)"}
   > RigVMVariableNode(Value=@n6.Item) {variable=Toe Pivot Control}
   n7 = &RigUnit_GetTransform(bInitial=true, Item=RigVMVariableNode() {variable=Ball Connector}.Value)
   n8 = &RigVMFunction_MathTransformMul(
    A=(Rotation=(X=0,Y=0,Z=0,W=1),Translation=(X=0,Y=0,Z=0),Scale3D=(X=1,Y=1,Z=1))
    A.Scale3D=@n7.Transform.Scale3D
    A.Translation=@n7.Transform.Translation
    B=RigVMVariableNode() {variable=Heel Roll Offset Transform}.Value
   ) {template="Multiply::Execute(in A,in B,out Result)"}
   Prepare Control Shape_3 = RigVMFunctionReferenceNode(
    Scale Premultiplier=0.2
    Control Shape=RigVMVariableNode() {variable=Heel Pivot Shape Settings}.Value
    Scale Factor=RigVMVariableNode() {variable=Control Scale}.Value
   )
   > n9 = &RigUnit_HierarchyAddControlTransform(
    InitialValue=(Rotation=(X=0,Y=0,Z=0,W=1),Translation=(X=0,Y=0,Z=0),Scale3D=(X=1,Y=1,Z=1))
    Name=Heel Pivot
    OffsetSpace=GlobalSpace
    OffsetTransform=(Rotation=(X=0,Y=0,Z=0,W=1),Translation=(X=0,Y=0,Z=0),Scale3D=(X=1,Y=1,Z=1))
    Settings="(InitialSpace=LocalSpace,bUsePreferredRotationOrder=False,PreferredRotationOrder=YZX,Limits=(LimitTranslationX=(bMinimum=False,bMaximum=False),LimitTranslationY=(bMinimum=False,bMaximum=False),LimitTranslationZ=(bMinimum=False,bMaximum=False),LimitPitch=(b ...(truncated)"
    OffsetTransform.Scale3D=@n8.Result.Scale3D
    OffsetTransform.Translation=@n8.Result.Translation
    Settings.Shape=@"Prepare Control Shape_3".Result Control Shape
    OffsetTransform.Rotation=&RigUnit_GetTransform(bInitial=False, Item=RigVMVariableNode() {variable=IK Foot Control}.Value).Transform.Rotation
    Parent=RigVMVariableNode() {variable=Toe Pivot Control}.Value
   ) {template="SpawnControl::Execute(in InitialValue,in Settings,in OffsetTransform,in OffsetSpace,in Parent,in Name,out Item)"}
   > RigVMVariableNode(Value=@n9.Item) {variable=Heel Pivot Control}
   > n10 = &RigUnit_HierarchyAddNull(
    Name=Heel Pivot Ball Aim Zero Ref
    Space=GlobalSpace
    Transform=&RigUnit_GetTransform(bInitial=False, Item=RigVMVariableNode() {variable=IK Foot Control}.Value).Transform
    Parent=@n9.Item
   )
   > RigVMVariableNode(Value=@n10.Item) {variable=Ball Aim Zero Ref Null}
   > n11 = &RigUnit_HierarchyAddNull(
    Name=Toe ROll
    Space=GlobalSpace
    Transform=&RigUnit_GetTransform(bInitial=true, Item=RigVMVariableNode() {variable=Toe Connector}.Value).Transform
    Parent=RigVMVariableNode() {variable=Heel Pivot Control}.Value
   )
   > RigVMVariableNode(Value=@n11.Item) {variable=Toe Roll Null}
   Greater = &RigVMFunction_MathDoubleGreater(B=0, A=RigVMVariableNode() {variable=IK Primary Axis}.Value.X) {template="Greater::Execute(in A,in B,out Result)"}
   Prepare Control Shape_4 = RigVMFunctionReferenceNode(
    Scale Premultiplier=0.75
    Scale Factor=RigVMVariableNode() {variable=Control Scale}.Value
    Control Shape=RigVMVariableNode() {variable=Toe Roll Shape Settings}.Value
   )
   > n12 = &RigUnit_HierarchyAddControlTransform(
    InitialValue=(Rotation=(X=0,Y=0,Z=0,W=1),Translation=(X=0,Y=0,Z=0),Scale3D=(X=1,Y=1,Z=1))
    Name=Toe Roll
    OffsetSpace=LocalSpace
    OffsetTransform=(Rotation=(X=-0.5,Y=-0.5,Z=-0.5,W=0.5),Translation=(X=0,Y=0,Z=0),Scale3D=(X=1,Y=1,Z=1))
    Settings="(InitialSpace=LocalSpace,bUsePreferredRotationOrder=False,PreferredRotationOrder=YZX,Limits=(LimitTranslationX=(bMinimum=False,bMaximum=False),LimitTranslationY=(bMinimum=False,bMaximum=False),LimitTranslationZ=(bMinimum=False,bMaximum=False),LimitPitch=(b ...(truncated)"
    Settings.Shape.Transform.Rotation=RigVMDispatchNode(
     Condition=@Greater.Result
     False=/Script/RigVM.RigVMFunction_MathQuaternionMul(B=(X=0,Y=0,Z=1,W=0), A=@"Prepare Control Shape_4".Result Control Shape.Transform.Rotation) {is=struct RigVMUnitNode, template="Multiply::Execute(in A,in B,out Result)"}.Result
     True=@"Prepare Control Shape_4".Result Control Shape.Transform.Rotation
    ) {template="DISPATCH_RigVMDispatch_If(in Condition,in True,in False,out Result)"}.Result
    Settings.Shape.Color=@"Prepare Control Shape_4".Result Control Shape.Color
    Settings.Shape.Name=@"Prepare Control Shape_4".Result Control Shape.Name
    Settings.Shape.Transform.Scale3D=@"Prepare Control Shape_4".Result Control Shape.Transform.Scale3D
    Settings.Shape.Transform.Translation=@"Prepare Control Shape_4".Result Control Shape.Transform.Translation
    Settings.Shape.bVisible=@"Prepare Control Shape_4".Result Control Shape.bVisible
    Parent=@n11.Item
   ) {template="SpawnControl::Execute(in InitialValue,in Settings,in OffsetTransform,in OffsetSpace,in Parent,in Name,out Item)"}
   > RigVMVariableNode(Value=@n12.Item) {variable=Toe Roll Control}
   > n13 = &RigUnit_HierarchyAddNull(
    Name=Ball Point
    Space=GlobalSpace
    Transform=&RigUnit_GetTransform(bInitial=true, Item=RigVMVariableNode() {variable=Ball Connector}.Value).Transform
    Parent=RigVMVariableNode() {variable=Toe Roll Control}.Value
   )
   > RigVMVariableNode(Value=@n13.Item) {variable=Ball Point Null}
   > &RigUnit_HierarchyAddNull(
    Name=Toe Point
    Space=GlobalSpace
    Transform=&RigUnit_GetTransform(bInitial=true, Item=RigVMVariableNode() {variable=Toe Connector}.Value).Transform
    Parent=RigVMVariableNode() {variable=Ball Point Null}.Value
   )
   > RigVMFunctionReturnNode()
   For_Each = RigVMDispatchNode(Array=RigVMVariableNode() {variable=IK Spaces}.Value, ExecuteContext=@RigVMFunction_Sequence.E) {template="DISPATCH_RigVMDispatch_ArrayIterator(in Array,out Element,out Index,out Count,out Ratio)"}
   > /Script/ControlRig.RigUnit_AddAvailableSpaces(
    Spaces=((Key=(Type=Control,Name="None"),Label="None"))
    Spaces.0.Key=@For_Each.Element
    Control=RigVMVariableNode() {variable=IK Foot Control}.Value
   ) {is=struct RigVMUnitNode}
   n14 = &RigVMFunction_ControlFlowBranch(Condition=@n1.Value, ExecuteContext=@RigVMFunction_Sequence.B)
   Prepare Control Shape = RigVMFunctionReferenceNode(
    Scale Premultiplier=1
    Control Shape=RigVMVariableNode() {variable=Hip Shape Settings}.Value
    Scale Factor=RigVMVariableNode() {variable=Control Scale}.Value
   )
   n15 = RigVMDispatchNode(
    Condition=@n1.Value
    False=RigVMVariableNode() {variable=Control Parent Connector}.Value
    True=RigVMVariableNode() {variable=Scapula Control}.Value
   ) {template="DISPATCH_RigVMDispatch_If(in Condition,in True,in False,out Result)"}
   n16 = &RigUnit_HierarchyAddControlTransform(
    InitialValue=(Rotation=(X=0,Y=0,Z=0,W=1),Translation=(X=0,Y=0,Z=0),Scale3D=(X=1,Y=1,Z=1))
    Name=Hip
    OffsetSpace=GlobalSpace
    OffsetTransform=(Rotation=(X=0,Y=0,Z=0,W=1),Translation=(X=0,Y=0,Z=0),Scale3D=(X=1,Y=1,Z=1))
    Settings="(InitialSpace=LocalSpace,bUsePreferredRotationOrder=False,PreferredRotationOrder=YZX,Limits=(LimitTranslationX=(bMinimum=False,bMaximum=False),LimitTranslationY=(bMinimum=False,bMaximum=False),LimitTranslationZ=(bMinimum=False,bMaximum=False),LimitPitch=(b ...(truncated)"
    Settings.Shape=@"Prepare Control Shape".Result Control Shape
    OffsetTransform.Translation=&RigUnit_GetTransform(bInitial=true, Item=RigVMVariableNode() {variable=Hip Connector}.Value).Transform.Translation
    ExecutePin=@n14.False
    Parent=@n15.Result
   ) {template="SpawnControl::Execute(in InitialValue,in Settings,in OffsetTransform,in OffsetSpace,in Parent,in Name,out Item)"}
   > RigVMVariableNode(Value=@n16.Item) {variable=Hip Control}
   > n17 = &RigUnit_HierarchyAddNull(Name=Hip, Space=LocalSpace, Transform=(Rotation=(X=0,Y=0,Z=0,W=1),Translation=(X=0,Y=0,Z=0),Scale3D=(X=1,Y=1,Z=1)), Parent=@n16.Item)
   > RigVMVariableNode(Value=@n17.Item) {variable=Hip Bottom Null}
   n18 = &RigUnit_HierarchyAddNull(
    Name=Hip PH
    Space=LocalSpace
    Transform=(Rotation=(X=0,Y=0,Z=0,W=1),Translation=(X=0,Y=0,Z=0),Scale3D=(X=1,Y=1,Z=1))
    ExecutePin=@n14.True
    Parent=@n15.Result
   )
   > RigVMVariableNode(Value=@n18.Item) {variable=Hip Control}
   Prepare Control Shape_0 = RigVMFunctionReferenceNode(
    Scale Premultiplier=1
    Control Shape=RigVMVariableNode() {variable=Ball FK Shape Settings}.Value
    Scale Factor=RigVMVariableNode() {variable=Control Scale}.Value
   )
   n19 = &RigUnit_HierarchyAddControlTransform(
    InitialValue=(Rotation=(X=0,Y=0,Z=0,W=1),Translation=(X=0,Y=0,Z=0),Scale3D=(X=1,Y=1,Z=1))
    Name=Ball FK
    OffsetSpace=GlobalSpace
    Settings="(InitialSpace=LocalSpace,bUsePreferredRotationOrder=False,PreferredRotationOrder=YZX,Limits=(LimitTranslationX=(bMinimum=False,bMaximum=False),LimitTranslationY=(bMinimum=False,bMaximum=False),LimitTranslationZ=(bMinimum=False,bMaximum=False),LimitPitch=(b ...(truncated)"
    Settings.Shape=@"Prepare Control Shape_0".Result Control Shape
    OffsetTransform=&RigUnit_GetTransform(bInitial=False, Item=RigVMVariableNode() {variable=Ball Connector}.Value).Transform
    Parent=RigVMVariableNode() {variable=Heel Pivot Control}.Value
    ExecutePin=@RigVMFunction_Sequence.D
   ) {template="SpawnControl::Execute(in InitialValue,in Settings,in OffsetTransform,in OffsetSpace,in Parent,in Name,out Item)"}
   > RigVMVariableNode(Value=@n19.Item) {variable=Ball FK Control}
   > n20 = &RigUnit_HierarchyAddNull(
    Name=Ball FK Tip Point
    Space=GlobalSpace
    Transform=&RigUnit_GetTransform(bInitial=False, Item=RigVMVariableNode() {variable=Toe Connector}.Value).Transform
    Parent=RigVMVariableNode() {variable=Ball FK Control}.Value
   )
   > RigVMVariableNode(Value=@n20.Item) {variable=Ball FK Tip Point Null}
   Prepare Control Shape_6 = RigVMFunctionReferenceNode(
    Scale Premultiplier=0.7
    Control Shape=RigVMVariableNode() {variable=Toe FK Shape Settings}.Value
    Scale Factor=RigVMVariableNode() {variable=Control Scale}.Value
   )
   > n21 = &RigUnit_HierarchyAddControlTransform(
    InitialValue=(Rotation=(X=0,Y=0,Z=0,W=1),Translation=(X=0,Y=0,Z=0),Scale3D=(X=1,Y=1,Z=1))
    Name=Toe FK
    OffsetSpace=GlobalSpace
    Settings="(InitialSpace=LocalSpace,bUsePreferredRotationOrder=False,PreferredRotationOrder=YZX,Limits=(LimitTranslationX=(bMinimum=False,bMaximum=False),LimitTranslationY=(bMinimum=False,bMaximum=False),LimitTranslationZ=(bMinimum=False,bMaximum=False),LimitPitch=(b ...(truncated)"
    Settings.Shape=@"Prepare Control Shape_6".Result Control Shape
    OffsetTransform=&RigUnit_GetTransform(bInitial=true, Item=RigVMVariableNode() {variable=Toe Connector}.Value).Transform
    Parent=RigVMVariableNode() {variable=Ball FK Tip Point Null}.Value
   ) {template="SpawnControl::Execute(in InitialValue,in Settings,in OffsetTransform,in OffsetSpace,in Parent,in Name,out Item)"}
   > RigVMVariableNode(Value=@n21.Item) {variable=Toe FK Control}
   > n22 = &RigUnit_HierarchyAddNull(
    Name=Ball FK Tip Point
    Space=GlobalSpace
    Transform=&RigUnit_GetTransform(bInitial=False, Item=RigVMVariableNode() {variable=Toe Tip Connector}.Value).Transform
    Parent=@n21.Item
   )
   > RigVMVariableNode(Value=@n22.Item) {variable=Toe FK Tip Point Null}
   RigVMDispatchNode(False=(X=-1,Y=1,Z=1), True=(X=1,Y=1,Z=1), Condition=@n2.Result) {template="DISPATCH_RigVMDispatch_If(in Condition,in True,in False,out Result)"}
   RigVMDispatchNode(
    False=(X=-1,Y=1,Z=1)
    True=(X=1,Y=1,Z=1)
    Condition=&RigVMFunction_MathDoubleGreater(B=0, A=RigVMVariableNode() {variable=IK Primary Axis}.Value.X) {template="Greater::Execute(in A,in B,out Result)"}.Result
   ) {template="DISPATCH_RigVMDispatch_If(in Condition,in True,in False,out Result)"}
   RigVMVariableNode() {variable=Scapula Shape Settings}
   RigVMVariableNode() {variable=Control Scale}
  ]}
 ConstructionGraph {nodes [
   Prepare Control Shape_0 = RigVMFunctionReferenceNode(
    Scale Premultiplier=2.5
    Scale Factor=RigVMVariableNode() {variable=Control Scale}.Value
    Control Shape=RigVMVariableNode() {variable=Knee Rotate Shape Settings}.Value
   )
   Control Stack at Position_2 = RigVMFunctionReferenceNode(
    Add Bottom Null=false
    Add Null=true
    Add Secondary Control=False
    Control Suffix=None
    ExecuteContext=()
    Global Position=(Rotation=(X=0,Y=0,Z=0,W=1),Translation=(X=0,Y=0,Z=0),Scale3D=(X=1,Y=1,Z=1))
    Name=Knee Rotate
    Null Suffix=None
    Orient to World=false
    Parent=(Type=None,Name="None")
    Secondary Settings="(Shape=(bVisible=True,Name=\"Default\",Color=(R=1,G=0,B=0,A=1),Transform=(Rotation=(X=0,Y=0,Z=0,W=1),Translation=(X=0,Y=0,Z=0),Scale3D=(X=1,Y=1,Z=1))),Proxy=(bIsProxy=False,ShapeVisibility=BasedOnSelection),DisplayName=\"None\",InitialSpace=LocalSpace,bUsePref ...(truncated)"
    Settings="(Shape=(bVisible=True,Name=\"Circle_Thin\",Color=(R=1,G=0,B=0,A=1),Transform=(Rotation=(X=0,Y=0,Z=0,W=1),Translation=(X=0,Y=0,Z=0),Scale3D=(X=2.5,Y=2.5,Z=2.5))),Proxy=(bIsProxy=False,ShapeVisibility=BasedOnSelection),DisplayName=\"None\",InitialSpace=LocalSpac ...(truncated)"
    Tag Controls=true
    Zero Controls=true
    Settings.Shape=@"Prepare Control Shape_0".Result Control Shape
   )
   > RigVMVariableNode(Value=@"Control Stack at Position_2".Control) {variable=Knee Rotator Control}
   > SpawnAnimationChannel = /Script/ControlRig.RigUnit_HierarchyAddAnimationChannelBool(
    InitialValue=False
    LimitsEnabled=()
    MaximumValue=True
    MinimumValue=False
    Name=Knee Rotator Vis
    Parent=RigVMVariableNode() {variable=Hip Control}.Value
   ) {
    is=struct RigVMUnitNode
    template="SpawnAnimationChannel::Execute(in InitialValue,in MinimumValue,in MaximumValue,in LimitsEnabled,in Parent,in Name,out Item)"
   }
   > RigVMVariableNode(Value=@SpawnAnimationChannel.Item) {variable=Knee Rotator Vis}
   RigUnit_PrepareForExecution = &RigUnit_PrepareForExecution()
   RigVMFunction_Sequence = RigVMAggregateNode(ExecuteContext=@RigUnit_PrepareForExecution.ExecutePin)
   Construct Foot = RigVMFunctionReferenceNode(ExecuteContext=@RigVMFunction_Sequence.B)
   > Initialize Variables = RigVMCollapseNode()
   Solve Knee PV Null = RigVMFunctionReferenceNode(Initial=true, End=RigVMVariableNode() {variable=Ball Connector}.Value, Thigh=RigVMVariableNode() {variable=Hip Connector}.Value)
   > RigVMVariableNode(Value=@"Solve Knee PV Null".Rotation) {variable=Initial Knee PV Null Orientation}
   > RigVMVariableNode(Value=@"Solve Knee PV Null".Result) {variable=Initial PV Null Location}
   Compute Pole Vector Location v02 = RigVMFunctionReferenceNode(
    Bones=((Type=None,Name="None"),(Type=None,Name="None"),(Type=None,Name="None"))
    Secondary Axis=(X=0,Y=0,Z=0)
    Bones.1=RigVMVariableNode() {variable=Knee Connector}.Value
    Bones.2=RigVMVariableNode() {variable=Ankle Connector}.Value
    PV Offset=RigVMVariableNode() {variable=PV Distance Scale}.Value
    Bones.0=RigVMVariableNode() {variable=Hip Connector}.Value
   )
   > RigVMVariableNode(Value=@"Compute Pole Vector Location v02".Pole Vector Location) {variable=Calculated PV Control Location}
   n1 = &RigUnit_HierarchyAddNull(
    Name=PV
    Parent=(Type=Bone,Name="None")
    Space=GlobalSpace
    Transform=(Rotation=(X=0,Y=0,Z=0,W=1),Translation=(X=0,Y=0,Z=0),Scale3D=(X=1,Y=1,Z=1))
    Transform.Scale3D=&RigUnit_GetTransform(bInitial=true, Item=RigVMVariableNode() {variable=Hip Connector}.Value).Transform.Scale3D
    Transform.Translation=RigVMVariableNode() {variable=Initial PV Null Location}.Value
    Transform.Rotation=RigVMVariableNode() {variable=Initial Knee PV Null Orientation}.Value
    ExecutePin=@RigVMFunction_Sequence.C
   )
   > RigVMVariableNode(Value=@n1.Item) {variable=Knee PV Null}
   RigUnit_GetTransform = &RigUnit_GetTransform(bInitial=true, Item=RigVMVariableNode() {variable=Hip Connector}.Value)
   n2 = /Script/ControlRig.RigUnit_AimBoneMath(
    DebugSettings=(bEnabled=False,Scale=10,WorldOffset=(Rotation=(X=0,Y=0,Z=0,W=1),Translation=(X=0,Y=0,Z=0),Scale3D=(X=1,Y=1,Z=1)))
    InputTransform=(Rotation=(X=0,Y=0,Z=0,W=1),Translation=(X=0,Y=0,Z=0),Scale3D=(X=1,Y=1,Z=1))
    Primary=(Weight=1,Axis=(X=0,Y=0,Z=1),Target=(X=1,Y=0,Z=0),Kind=Location,Space=(Type=Bone,Name="None"))
    Secondary=(Weight=1,Axis=(X=1,Y=0,Z=0),Target=(X=1,Y=0,Z=0),Kind=Direction,Space=(Type=Bone,Name="None"))
    Weight=1
    Primary.Target=@RigUnit_GetTransform.Transform.Translation
    InputTransform.Translation=&RigUnit_GetTransform(bInitial=true, Item=RigVMVariableNode() {variable=Ankle Connector}.Value).Transform.Translation
   ) {is=struct RigVMUnitNode}
   Prepare Control Shape = RigVMFunctionReferenceNode(
    Scale Premultiplier=0.3
    Scale Factor=RigVMVariableNode() {variable=Control Scale}.Value
    Control Shape=RigVMVariableNode() {variable=PV Shape Settings}.Value
   )
   > SpawnControl = &RigUnit_HierarchyAddControlTransform(
    InitialValue=(Rotation=(X=0,Y=0,Z=0,W=1),Translation=(X=0,Y=0,Z=0),Scale3D=(X=1,Y=1,Z=1))
    Name=PV
    OffsetSpace=GlobalSpace
    OffsetTransform=(Rotation=(X=0,Y=0,Z=0,W=1),Translation=(X=0,Y=0,Z=0),Scale3D=(X=1,Y=1,Z=1))
    Settings="(InitialSpace=LocalSpace,bUsePreferredRotationOrder=False,PreferredRotationOrder=YZX,Limits=(LimitTranslationX=(bMinimum=False,bMaximum=False),LimitTranslationY=(bMinimum=False,bMaximum=False),LimitTranslationZ=(bMinimum=False,bMaximum=False),LimitPitch=(b ...(truncated)"
    OffsetTransform.Rotation=@n2.Result.Rotation
    Settings.Shape=@"Prepare Control Shape".Result Control Shape
    OffsetTransform.Translation=RigVMVariableNode() {variable=Calculated PV Control Location}.Value
    Parent=@n1.Item
   ) {template="SpawnControl::Execute(in InitialValue,in Settings,in OffsetTransform,in OffsetSpace,in Parent,in Name,out Item)"}
   > RigVMVariableNode(Value=@SpawnControl.Item) {variable=Knee PV Control}
   Create Sec Control Visibility Switch = RigVMFunctionReferenceNode(ExecuteContext=@RigVMFunction_Sequence.F)
   Construct Ankle Roll Control = RigVMCollapseNode(Parent=RigVMVariableNode() {variable=Ball Point Null}.Value, ExecutePin=@RigVMFunction_Sequence.E)
   Extrapolate Quaternion = RigVMFunctionReferenceNode(MaxAngle=RigVMVariableNode() {variable=Hock Angle Sweep}.Value, Target=RigVMVariableNode() {variable=Calculated PV Control Location}.Value)
   RigVMVariableNode(Value=@"Extrapolate Quaternion".Initial T, ExecuteContext=@RigVMFunction_Sequence.D) {variable=Ankle Initial Interp Target}
   > RigVMVariableNode(Value=@"Extrapolate Quaternion".Squashed Relative) {variable=Local Squash Target}
   RigVMVariableNode(
    Value=&RigUnit_ResolveConnector(Connector=(Type=Connector,Name="IK Parent"), SkipSocket=False).Result
    ExecuteContext=@RigVMFunction_Sequence.A
   ) {variable=Pelvis Connector}
   > RigVMVariableNode(Value=&RigUnit_ResolveConnector(Connector=(Type=Connector,Name="Start"), SkipSocket=False).Result) {variable=Hip Connector}
   > RigVMVariableNode(Value=&RigUnit_ResolveConnector(Connector=(Type=Connector,Name="Mid"), SkipSocket=False).Result) {variable=Knee Connector}
   > RigVMVariableNode(Value=&RigUnit_ResolveConnector(Connector=(Type=Connector,Name="End"), SkipSocket=False).Result) {variable=Ankle Connector}
   > RigVMVariableNode(Value=&RigUnit_ResolveConnector(Connector=(Type=Connector,Name="Ball"), SkipSocket=False).Result) {variable=Ball Connector}
   > RigVMVariableNode(Value=&RigUnit_ResolveConnector(Connector=(Type=Connector,Name="Toe"), SkipSocket=False).Result) {variable=Toe Connector}
   n3 = &RigUnit_ResolveConnector(Connector=(Type=Connector,Name="Toe Tip"), SkipSocket=False)
   > RigVMVariableNode(Value=@n3.Result) {variable=Toe Tip Connector}
   > RigVMVariableNode(Value=@n3.bIsConnected) {variable=Is Toe Tip Connected}
   > RigVMVariableNode(Value=&RigUnit_ResolveConnector(Connector=(Type=Connector,Name="Parent"), SkipSocket=False).Result) {variable=Control Parent Connector}
   > RigVMVariableNode(Value=/Script/ControlRig.RigUnit_ResolveArrayConnector(Connector=(Type=Connector,Name="IK Spaces"), SkipSocket=False) {is=struct RigVMUnitNode}.Result) {variable=IK Spaces}
   n4 = &RigUnit_ResolveConnector(Connector=(Type=Connector,Name="Scapula"), SkipSocket=False)
   > RigVMVariableNode(Value=@n4.Result) {variable=Scapula Connector}
   > RigVMVariableNode(Value=@n4.bIsConnected) {variable=Is Scapula Connected}
   /Script/ControlRig.RigUnit_SetChannelHosts(
    ExecutePin=()
    Hosts=((Type=None,Name="None"),(Type=None,Name="None"))
    Channel=RigVMVariableNode() {variable=Knee Rotator Vis}.Value
    Hosts.0=RigVMVariableNode() {variable=Knee PV Control}.Value
    Hosts.1=RigVMVariableNode() {variable=IK Foot Control}.Value
   ) {is=struct RigVMUnitNode}
  ]}
 Multiply_1_ContainedGraph {nodes [
   n1 = RigVMFunctionEntryNode()
   Multiply = &RigVMFunction_MathDoubleMul(A=@n1.A, B=@n1.B)
   RigVMFunctionReturnNode(Result=&RigVMFunction_MathDoubleMul(B=@n1.C, A=@Multiply.Result).Result)
  ]}
 New Function_ContainedGraph {nodes [
   RigVMFunctionEntryNode()
   n1 = /Script/RigVM.RigVMFunction_MathDoubleRad(Value=RigVMVariableNode() {variable=MaxAngle}.Value) {is=struct RigVMUnitNode, template="Radians::Execute(in Value,out Result)"}
   n2 = RigVMVariableNode() {variable=Initial}
   Make = RigVMDispatchNode(Elements=(X=0,Y=0,Z=1)) {template="DISPATCH_RigVMDispatch_MakeStruct(in Elements,out Struct)"}
   Negate = &RigVMFunction_MathVectorNegate(Value=@Make.Struct) {template="Negate::Execute(in Value,out Result)"}
   n3 = RigVMDispatchNode(True=@Negate.Result, Condition=RigVMVariableNode() {variable=Invert Heel Secondary Axis}.Value, False=@Make.Struct) {template="DISPATCH_RigVMDispatch_If(in Condition,in True,in False,out Result)"}
   n4 = &RigUnit_AimBoneMath(
    DebugSettings=(bEnabled=False,Scale=10,WorldOffset=(Rotation=(X=0,Y=0,Z=0,W=1),Translation=(X=0,Y=0,Z=0),Scale3D=(X=1,Y=1,Z=1)))
    Primary=(Weight=1,Axis=(X=1,Y=0,Z=0),Target=(X=1,Y=0,Z=0),Kind=Location,Space=(Type=Bone,Name="None"))
    Secondary=(Weight=1,Axis=(X=0,Y=0,Z=1),Target=(X=1,Y=0,Z=0),Kind=Location,Space=(Type=Bone,Name="None"))
    Weight=1
    InputTransform=&RigUnit_GetTransform(bInitial=@n2.Value, Item=RigVMVariableNode() {variable=Ball Connector}.Value).Transform
    Primary.Target=&RigUnit_GetTransform(bInitial=@n2.Value, Item=RigVMVariableNode() {variable=Hip Connector}.Value).Transform.Translation
    Secondary.Target=RigVMVariableNode() {variable=Target}.Value
    Secondary.Axis=@n3.Result
   )
   n5 = RigVMVariableNode() {variable=Initial}
   n6 = &RigUnit_AimBoneMath(
    DebugSettings=(bEnabled=False,Scale=10,WorldOffset=(Rotation=(X=0,Y=0,Z=0,W=1),Translation=(X=0,Y=0,Z=0),Scale3D=(X=1,Y=1,Z=1)))
    Primary=(Weight=1,Axis=(X=1,Y=0,Z=0),Target=(X=1,Y=0,Z=0),Kind=Location,Space=(Type=Bone,Name="None"))
    Secondary=(Weight=1,Axis=(X=0,Y=0,Z=1),Target=(X=0,Y=0,Z=0),Kind=Location,Space=(Type=Bone,Name="None"))
    Weight=1
    InputTransform=&RigUnit_GetTransform(bInitial=@n5.Value, Item=RigVMVariableNode() {variable=Ball Connector}.Value).Transform
    Primary.Target=&RigUnit_GetTransform(bInitial=@n5.Value, Item=RigVMVariableNode() {variable=Ankle Connector}.Value).Transform.Translation
    Secondary.Target=RigVMVariableNode() {variable=Target}.Value
    Secondary.Axis=@n3.Result
   )
   n7 = /Script/RigVM.RigVMFunction_MathQuaternionToAxisAndAngle(Value=/Script/RigVM.RigVMFunction_MathQuaternionMakeRelative(Parent=@n4.Result.Rotation, Global=@n6.Result.Rotation) {is=struct RigVMUnitNode, template="Make Relative::Execute(in Global,in Parent,out Local)"}.Local) {is=struct RigVMUnitNode, template="ToAxisAndAngle::Execute(in Value,out Axis,out Angle)"}
   n8 = &RigVMFunction_MathDoubleDiv(B=@n1.Result, A=@n7.Angle)
   OneMinus = /Script/RigVM.RigVMFunction_MathDoubleOneMinus(Value=@n8.Result) {is=struct RigVMUnitNode, template="OneMinus::Execute(in Value,out Result)"}
   n9 = RigVMVariableNode() {variable=Invert Heel Secondary Axis}
   n10 = /Script/RigVM.RigVMFunction_MathQuaternionFromAxisAndAngle(
    Axis=RigVMDispatchNode(
     True=&RigVMFunction_MathVectorNegate(Value=@n7.Axis) {template="Negate::Execute(in Value,out Result)"}.Result
     Condition=@n9.Value
     False=@n7.Axis
    ) {template="DISPATCH_RigVMDispatch_If(in Condition,in True,in False,out Result)"}.Result
    Angle=@n1.Result
   ) {is=struct RigVMUnitNode, template="FromAxisAndAngle::Execute(in Axis,in Angle,out Result)"}
   > RigVMFunctionReturnNode(
    Initial T=RigVMDispatchNode(True=@OneMinus.Result, Condition=@n9.Value, False=@n8.Result) {template="DISPATCH_RigVMDispatch_If(in Condition,in True,in False,out Result)"}.Result
    Squashed Relative=@n10.Result
   )
   /Script/RigVM.RigVMFunction_MathQuaternionMakeAbsolute(Parent=@n4.Result.Rotation, Local=@n10.Result) {is=struct RigVMUnitNode, template="Make Absolute::Execute(in Local,in Parent,out Global)"}
   RigVMRerouteNode(Value=@n4.Result.Rotation)
  ]}
 New Function_ContainedGraph {nodes [
   n1 = &RigUnit_GetTransform(Item=RigVMVariableNode() {variable=End}.Value, bInitial=RigVMVariableNode() {variable=Initial}.Value)
   RigUnit_GetTransform = &RigUnit_GetTransform(Space=LocalSpace, bInitial=False, Item=RigVMVariableNode() {variable=Knee PV Null}.Value)
   Interpolate = /Script/RigVM.RigVMFunction_MathVectorLerp(
    T=0.5
    B=/Script/RigVM.RigVMFunction_MathTransformMakeAbsolute(Local=(Rotation=(X=0,Y=0,Z=0,W=1),Translation=(X=0,Y=5,Z=0),Scale3D=(X=1,Y=1,Z=1)), Parent=@RigUnit_GetTransform.Transform) {is=struct RigVMUnitNode, template="Make Absolute::Execute(in Local,in Parent,out Global)"}.Global.Translation
    A=&RigUnit_GetTransform(bInitial=False, Item=RigVMVariableNode() {variable=Hip Control}.Value).Transform.Translation
   ) {is=struct RigVMUnitNode, template="Interpolate::Execute(in A,in B,in T,out Result)"}
   n2 = &RigUnit_AimBoneMath(
    DebugSettings=(bEnabled=False,Scale=10,WorldOffset=(Rotation=(X=0,Y=0,Z=0,W=1),Translation=(X=0,Y=0,Z=0),Scale3D=(X=1,Y=1,Z=1)))
    Primary=(Weight=1,Axis=(X=1,Y=0,Z=0),Target=(X=1,Y=0,Z=0),Kind=Location,Space=(Type=Bone,Name="None"))
    Secondary=(Weight=1,Axis=(X=0,Y=0,Z=1),Target=(X=0,Y=0,Z=0),Kind=Location,Space=(Type=Bone,Name="None"))
    Weight=1
    InputTransform=@n1.Transform
    Primary.Target=&RigUnit_GetTransform(bInitial=False, Item=RigVMVariableNode() {variable=Base}.Value).Transform.Translation
    Secondary.Axis=RigVMDispatchNode(False=(X=0,Y=0,Z=1), True=(X=0,Y=0,Z=-1), Condition=RigVMVariableNode() {variable=Invert Heel Secondary Axis}.Value) {template="DISPATCH_RigVMDispatch_If(in Condition,in True,in False,out Result)"}.Result
    Secondary.Target=@Interpolate.Result
   )
   n3 = RigVMVariableNode() {variable=Local Squash Target}
   n4 = RigVMVariableNode() {variable=Ankle Initial Interp Target}
   n5 = RigVMVariableNode() {variable=Chain Length}
   Subtract = &RigVMFunction_MathDoubleSub(A=RigVMVariableNode() {variable=Initial Distance Hip to IK Target}.Value, B=@n5.Value)
   Add = &RigVMFunction_MathDoubleAdd(
    A=&RigVMFunction_MathDoubleMul(A=&RigVMFunction_MathDoubleDiv(A=1, B=@n4.Value).Result, B=@Subtract.Result).Result
    B=@n5.Value
   )
   RigVMFunction_MathVectorDistance = &RigVMFunction_MathVectorDistance(
    B=@n1.Transform.Translation
    A=&RigUnit_GetTransform(Item=RigVMVariableNode() {variable=Base}.Value, bInitial=RigVMVariableNode() {variable=Initial}.Value).Transform.Translation
   )
   Remap = &RigVMFunction_MathDoubleRemap(TargetMaximum=0, TargetMinimum=1, SourceMinimum=@Add.Result, Value=@RigVMFunction_MathVectorDistance.Result, SourceMaximum=@n5.Value)
   n6 = &RigUnit_GetTransform(bInitial=RigVMVariableNode() {variable=Initial}.Value, Item=RigVMVariableNode() {variable=Base}.Value)
   n7 = RigVMVariableNode() {variable=End}
   n8 = /Script/RigVM.RigVMFunction_MathVectorMakeRelative(
    Parent=@n6.Transform.Translation
    Global=/Script/ControlRig.RigUnit_ProjectTransformToNewParent(
     bChildInitial=True
     bNewParentInitial=False
     bOldParentInitial=True
     NewParent=@n7.Value
     OldParent=@n7.Value
     Child=RigVMVariableNode() {variable=Ankle}.Value
    ) {is=struct RigVMUnitNode}.Transform.Translation
   ) {is=struct RigVMUnitNode, template="Make Relative::Execute(in Global,in Parent,out Local)"}
   RigVMFunction_MathVectorLength = &RigVMFunction_MathVectorLength(Value=@n8.Local) {template="Length::Execute(in Value,out Result)"}
   Minimum = /Script/RigVM.RigVMFunction_MathDoubleMin(A=@RigVMFunction_MathVectorLength.Result, B=RigVMVariableNode() {variable=Initial Thigh to Ankle Length}.Value) {is=struct RigVMUnitNode, template="Minimum::Execute(in A,in B,out Result)"}
   RigVMFunction_MathVectorSetLength = /Script/RigVM.RigVMFunction_MathVectorSetLength(Value=@n8.Local, Length=@Minimum.Result) {is=struct RigVMUnitNode, template="SetLength::Execute(in Value,in Length,out Result)"}
   RigVMFunctionReturnNode(
    Result=(Rotation=(X=0,Y=0,Z=0,W=1),Translation=(X=0,Y=0,Z=0),Scale3D=(X=1,Y=1,Z=1))
    Result.Rotation=/Script/RigVM.RigVMFunction_MathQuaternionSlerp(
     A=@n2.Result.Rotation
     B=&RigVMFunction_MathQuaternionMakeAbsolute(Parent=@n2.Result.Rotation, Local=@n3.Value) {template="Make Absolute::Execute(in Local,in Parent,out Global)"}.Global
     T=@Remap.Result
    ) {is=struct RigVMUnitNode, template="Interpolate::Execute(in A,in B,in T,out Result)"}.Result
    Result.Translation=/Script/RigVM.RigVMFunction_MathVectorMakeAbsolute(Local=@RigVMFunction_MathVectorSetLength.Result, Parent=@n6.Transform.Translation) {is=struct RigVMUnitNode, template="Make Absolute::Execute(in Local,in Parent,out Global)"}.Global
    Execute=RigVMFunctionEntryNode().Execute
   )
  ]}
 New Function_ContainedGraph {nodes [
   n1 = RigVMFunctionEntryNode()
   n2 = &RigUnit_GetTransform(bInitial=False, Item=RigVMVariableNode() {variable=Hip Connector}.Value)
   ToVectors = /Script/RigVM.RigVMFunction_MathQuaternionToVectors(Value=@n2.Transform.Rotation) {is=struct RigVMUnitNode, template="ToVectors::Execute(in Value,out Forward,out Right,out Up)"}
   RigVMFunction_MathVectorCross = /Script/RigVM.RigVMFunction_MathVectorCross(
    B=&RigVMFunction_MathVectorNegate(Value=@ToVectors.Forward) {template="Negate::Execute(in Value,out Result)"}.Result
    A=@n2.Transform.Translation
   ) {is=struct RigVMUnitNode, template="Cross::Execute(in A,in B,out Result)"}
   n3 = RigVMVariableNode() {variable=Toe Connector}
   n4 = &RigUnit_AimBoneMath(
    DebugSettings=(bEnabled=False,Scale=10,WorldOffset=(Rotation=(X=0,Y=0,Z=0,W=1),Translation=(X=0,Y=0,Z=0),Scale3D=(X=1,Y=1,Z=1)))
    Primary=(Weight=1,Axis=(X=1,Y=0,Z=0),Target=(X=1,Y=0,Z=0),Kind=Location,Space=(Type=Bone,Name="None"))
    Secondary=(Weight=1,Axis=(X=0,Y=0,Z=1),Target=(X=1,Y=0,Z=0),Kind=Direction,Space=(Type=Bone,Name="None"))
    Weight=1
    Secondary.Target=@RigVMFunction_MathVectorCross.Result
    InputTransform=&RigUnit_GetTransform(bInitial=False, Item=@n3.Value).Transform
    Primary.Target=@n2.Transform.Translation
   )
   n5 = &RigUnit_GetTransform(bInitial=true, Item=@n3.Value)
   n6 = RigVMVariableNode() {variable=Hip Connector}
   n7 = RigVMVariableNode() {variable=Toe Connector}
   n8 = &RigVMFunction_MathDoubleRemap(
    TargetMaximum=1
    TargetMinimum=0
    Value=&RigVMFunction_MathVectorDistance(
     B=&RigUnit_GetTransform(bInitial=False, Item=@n6.Value).Transform.Translation
     A=&RigUnit_GetTransform(bInitial=False, Item=@n7.Value).Transform.Translation
    ).Result
    SourceMinimum=RigVMVariableNode() {variable=Hock Chain Length}.Value
    SourceMaximum=RigVMVariableNode() {variable=Chain Length}.Value
   )
   RigVMFunction_MathBoolNot = &RigVMFunction_MathBoolNot(Value=@n1.Backwards Solve)
   Add = &RigVMFunction_MathDoubleAdd(A=RigVMVariableNode() {variable=Chain Length}.Value, B=RigVMVariableNode() {variable=Initial Ball Length}.Value)
   RigVMFunction_MathVectorDistance = &RigVMFunction_MathVectorDistance(
    B=&RigUnit_GetTransform(bInitial=False, Item=RigVMVariableNode() {variable=IK Foot Control}.Value).Transform.Translation
    A=&RigUnit_GetTransform(bInitial=False, Item=RigVMVariableNode() {variable=Hip Control}.Value).Transform.Translation
   )
   Remap = &RigVMFunction_MathDoubleRemap(
    TargetMaximum=1
    TargetMinimum=0
    SourceMaximum=@Add.Result
    SourceMinimum=RigVMVariableNode() {variable=Hock Chain Length}.Value
    Value=@RigVMFunction_MathVectorDistance.Result
   )
   n9 = RigVMVariableNode() {variable=IK Secondary Axis}
   Negate = &RigVMFunction_MathVectorNegate(Value=RigVMVariableNode() {variable=IK Primary Axis}.Value) {template="Negate::Execute(in Value,out Result)"}
   n10 = &RigUnit_AimBoneMath(
    DebugSettings=(bEnabled=False,Scale=10,WorldOffset=(Rotation=(X=0,Y=0,Z=0,W=1),Translation=(X=0,Y=0,Z=0),Scale3D=(X=1,Y=1,Z=1)))
    Primary=(Weight=1,Axis=(X=-1,Y=0,Z=0),Target=(X=1,Y=0,Z=0),Kind=Location,Space=(Type=Bone,Name="None"))
    Secondary=(Weight=1,Axis=(X=0,Y=1,Z=0),Target=(X=1,Y=0,Z=0),Kind=Location,Space=(Type=Bone,Name="None"))
    Weight=1
    InputTransform=&RigUnit_GetTransform(bInitial=False, Item=RigVMVariableNode() {variable=IK Foot Control}.Value).Transform
    Primary.Target=&RigUnit_GetTransform(bInitial=False, Item=RigVMVariableNode() {variable=Hip Control}.Value).Transform.Translation
    Secondary.Axis=RigVMDispatchNode(
     True=&RigVMFunction_MathVectorNegate(Value=@n9.Value) {template="Negate::Execute(in Value,out Result)"}.Result
     Condition=RigVMVariableNode() {variable=Invert Heel Secondary Axis}.Value
     False=@n9.Value
    ) {template="DISPATCH_RigVMDispatch_If(in Condition,in True,in False,out Result)"}.Result
    Primary.Axis=@Negate.Result
    Secondary.Space=RigVMVariableNode() {variable=Knee PV Control}.Value
   )
   Interpolate = &RigVMFunction_MathQuaternionSlerp(
    A=&RigUnit_GetTransform(bInitial=true, Item=RigVMVariableNode() {variable=Toe Roll Null}.Value).Transform.Rotation
    T=@Remap.Result
    B=@n10.Result.Rotation
   ) {template="Interpolate::Execute(in A,in B,in T,out Result)"}
   n11 = RigVMVariableNode() {variable=Ball Aim Zero Ref Null}
   Multiply = &RigVMFunction_MathQuaternionMul(
    B=@Interpolate.Result
    A=&RigVMFunction_MathQuaternionMakeRelative(
     Parent=&RigUnit_GetTransform(bInitial=true, Item=@n11.Value).Transform.Rotation
     Global=&RigUnit_GetTransform(bInitial=false, Item=@n11.Value).Transform.Rotation
    ) {template="Make Relative::Execute(in Global,in Parent,out Local)"}.Local
   ) {template="Multiply::Execute(in A,in B,out Result)"}
   > RigVMFunctionReturnNode(Result=RigVMDispatchNode(
     False=&RigVMFunction_MathQuaternionSlerp(B=@n4.Result.Rotation, A=@n5.Transform.Rotation, T=@n8.Result) {template="Interpolate::Execute(in A,in B,in T,out Result)"}.Result
     Condition=@RigVMFunction_MathBoolNot.Result
     True=@Multiply.Result
    ) {template="DISPATCH_RigVMDispatch_If(in Condition,in True,in False,out Result)"}.Result)
  ]}
 New Function_ContainedGraph {nodes [
   RigVMFunctionEntryNode()
   RigUnit_GetTransform = &RigUnit_GetTransform(bInitial=False, Item=RigVMVariableNode() {variable=Hip Control}.Value)
   ToVectors = &RigVMFunction_MathQuaternionToVectors(Value=@RigUnit_GetTransform.Transform.Rotation) {template="ToVectors::Execute(in Value,out Forward,out Right,out Up)"}
   n1 = &RigUnit_AimBoneMath(
    DebugSettings=(bEnabled=False,Scale=10,WorldOffset=(Rotation=(X=0,Y=0,Z=0,W=1),Translation=(X=0,Y=0,Z=0),Scale3D=(X=1,Y=1,Z=1)))
    Primary=(Weight=1,Axis=(X=1,Y=0,Z=0),Target=(X=1,Y=0,Z=0),Kind=Location,Space=(Type=Bone,Name="None"))
    Secondary=(Weight=1,Axis=(X=0,Y=0,Z=1),Target=(X=-1,Y=0,Z=0),Kind=Direction,Space=(Type=Bone,Name="None"))
    Weight=1
    InputTransform=&RigUnit_GetTransform(Item=RigVMVariableNode() {variable=End}.Value, bInitial=RigVMVariableNode() {variable=Initial}.Value).Transform
    Primary.Target=&RigUnit_GetTransform(Item=RigVMVariableNode() {variable=Thigh}.Value, bInitial=RigVMVariableNode() {variable=Initial}.Value).Transform.Translation
    Secondary.Target=@ToVectors.Forward
   )
   Interpolate = &RigVMFunction_MathVectorLerp(
    B=&RigUnit_GetTransform(Item=RigVMVariableNode() {variable=End}.Value, bInitial=RigVMVariableNode() {variable=Initial}.Value).Transform.Translation
    A=&RigUnit_GetTransform(Item=RigVMVariableNode() {variable=Thigh}.Value, bInitial=RigVMVariableNode() {variable=Initial}.Value).Transform.Translation
    T=RigVMVariableNode() {variable=PV Null Ratio}.Value
   ) {template="Interpolate::Execute(in A,in B,in T,out Result)"}
   > RigVMFunctionReturnNode(Rotation=@n1.Result.Rotation, Result=@Interpolate.Result)
  ]}
 New Function_ContainedGraph {nodes [
   RigVMFunctionEntryNode()
   n1 = &RigVMFunction_MathTransformMakeRelative(Parent=RigVMVariableNode() {variable=Root}.Value, Global=RigVMVariableNode() {variable=Transform}.Value)
   RigVMFunction_MathVectorLength = &RigVMFunction_MathVectorLength(Value=@n1.Local.Translation) {template="Length::Execute(in Value,out Result)"}
   Minimum = &RigVMFunction_MathDoubleMin(A=@RigVMFunction_MathVectorLength.Result, B=RigVMVariableNode() {variable=Length}.Value) {template="Minimum::Execute(in A,in B,out Result)"}
   RigVMFunction_MathVectorSetLength = &RigVMFunction_MathVectorSetLength(Value=@n1.Local.Translation, Length=@Minimum.Result) {template="SetLength::Execute(in Value,in Length,out Result)"}
   > RigVMFunctionReturnNode(Result=&RigVMFunction_MathTransformMakeAbsolute(
     Local=(Rotation=(X=0,Y=0,Z=0,W=1),Translation=(X=0,Y=0,Z=0),Scale3D=(X=1,Y=1,Z=1))
     Local.Translation=@RigVMFunction_MathVectorSetLength.Result
     Parent=RigVMVariableNode() {variable=Root}.Value
     Local.Rotation=@n1.Local.Rotation
     Local.Scale3D=@n1.Local.Scale3D
    ) {template="Make Absolute::Execute(in Local,in Parent,out Global)"}.Global)
  ]}
 New Function_ContainedGraph {nodes [
   n1 = RigVMFunctionEntryNode()
   n2 = /Script/ControlRig.RigUnit_GetModuleName() {is=struct RigVMUnitNode}
   EndsWith = /Script/RigVM.RigVMFunction_StringEndsWith(Ending=" L", Name=@n2.Module) {is=struct RigVMUnitNode, template="EndsWith::Execute(in Name,in Ending,out Result)"}
   n3 = &RigVMFunction_StringEndsWith(Ending=" R", Name=@n2.Module) {template="EndsWith::Execute(in Name,in Ending,out Result)"}
   Get_Module_Metadata = RigVMDispatchNode(
    NameSpace=Root
    Name=RigVMDispatchNode(
     True=Global Left Control Color
     Condition=@EndsWith.Result
     False=RigVMDispatchNode(False=Global Center Control Color, True=Global Right Control Color, Condition=@n3.Result) {template="DISPATCH_RigVMDispatch_If(in Condition,in True,in False,out Result)"}.Result
    ) {template="DISPATCH_RigVMDispatch_If(in Condition,in True,in False,out Result)"}.Result
    Default=@n1.Control Shape.Color
   ) {template="DISPATCH_RigDispatch_GetModuleMetadata(in Name,in NameSpace,in Default,out Value,out Found)"}
   Interpolate = /Script/RigVM.RigVMFunction_MathColorLerp(B=@Get_Module_Metadata.Value, T=RigVMVariableNode() {variable=Root Module Color Weight}.Value, A=@n1.Control Shape.Color) {is=struct RigVMUnitNode, template="Interpolate::Execute(in A,in B,in T,out Result)"}
   Get_Module_Metadata_3 = RigVMDispatchNode(Default=1, Name=Global Control Scale, NameSpace=Root) {template="DISPATCH_RigDispatch_GetModuleMetadata(in Name,in NameSpace,in Default,out Value,out Found)"}
   Multiply = RigVMAggregateNode(
    C=@Get_Module_Metadata_3.Value
    A=RigVMVariableNode() {variable=Scale Factor}.Value
    B=RigVMVariableNode() {variable=Scale Premultiplier}.Value
   )
   n4 = RigVMVariableNode() {variable=Control Shape}
   Scale = /Script/RigVM.RigVMFunction_MathVectorScale(Factor=@Multiply.Result, Value=@n4.Value.Transform.Scale3D) {is=struct RigVMUnitNode, template="Scale::Execute(in Value,in Factor,out Result)"}
   > RigVMFunctionReturnNode(
    Result Control Shape=(bVisible=True,Name="Default",Color=(R=1,G=0,B=0,A=1),Transform=(Rotation=(X=0,Y=0,Z=0,W=1),Translation=(X=0,Y=0,Z=0),Scale3D=(X=1,Y=1,Z=1)))
    Result Control Shape.Name=@n1.Control Shape.Name
    Result Control Shape.bVisible=@n1.Control Shape.bVisible
    Result Control Shape.Color=@Interpolate.Result
    Result Control Shape.Transform.Scale3D=@Scale.Result
    Result Control Shape.Transform.Rotation=@n4.Value.Transform.Rotation
    Result Control Shape.Transform.Translation=@n4.Value.Transform.Translation
   )
  ]}
 New Function_ContainedGraph {nodes [
   n1 = RigVMFunctionEntryNode()
   RigVMFunction_Sequence = &RigVMFunction_Sequence(ExecuteContext=@n1.ExecuteContext)
   RigVMFunctionReturnNode(ExecuteContext=@RigVMFunction_Sequence.B)
   n2 = &RigUnit_HierarchyAddAnimationChannelBool(
    InitialValue=false
    LimitsEnabled=()
    MaximumValue=True
    MinimumValue=false
    Name=Sec Control Visibility
    Parent=RigVMVariableNode() {variable=IK Foot Control}.Value
    ExecutePin=@RigVMFunction_Sequence.A
   ) {template="SpawnAnimationChannel::Execute(in InitialValue,in MinimumValue,in MaximumValue,in LimitsEnabled,in Parent,in Name,out Item)"}
   > RigVMVariableNode(Value=@n2.Item) {variable=Sec Control Vis Channel}
   > &RigUnit_SetChannelHosts(Hosts=((Type=None,Name="None")), Hosts.0=RigVMVariableNode() {variable=Toe FK Control}.Value, Channel=@n2.Item)
  ]}
 New Function_ContainedGraph {nodes [
   n1 = RigVMFunctionEntryNode()
   RigVMFunction_Sequence = &RigVMFunction_Sequence(ExecuteContext=@n1.ExecuteContext)
   RigVMFunctionReturnNode(ExecuteContext=@RigVMFunction_Sequence.B)
   For_Each = RigVMDispatchNode(
    Array=((Type=None,Name="None"),(Type=None,Name="None"))
    Array.0=RigVMVariableNode() {variable=Toe FK Control}.Value
    Array.1=RigVMVariableNode() {variable=Ball FK Control}.Value
    ExecuteContext=@RigVMFunction_Sequence.A
   ) {template="DISPATCH_RigVMDispatch_ArrayIterator(in Array,out Element,out Index,out Count,out Ratio)"}
   > SetControlVisibility = /Script/ControlRig.RigUnit_SetControlVisibility(
    Item=@For_Each.Element
    bVisible=/Script/ControlRig.RigUnit_GetBoolAnimationChannelFromItem(bInitial=False, Item=RigVMVariableNode() {variable=Sec Control Vis Channel}.Value) {is=struct RigVMUnitNode, template="GetAnimationChannelFromItem::Execute(out Value,in Item,in bInitial)"}.Value
   ) {is=struct RigVMUnitNode, template="SetControlVisibility::Execute(in Item,in Pattern,in bVisible)"}
  ]}
 RigVMFunctionLibrary {nodes [
   Construct Foot = RigVMCollapseNode(ExecuteContext=())
   Create Sec Control Visibility Switch = RigVMCollapseNode(ExecuteContext=())
   Extrapolate Quaternion = RigVMCollapseNode(Initial=true, MaxAngle=0, Target=(X=1,Y=0,Z=0))
   Limit Transform Length = RigVMCollapseNode(
    Length=10
    Root=(Rotation=(X=0,Y=0,Z=0,W=1),Translation=(X=0,Y=0,Z=0),Scale3D=(X=1,Y=1,Z=1))
    Transform=(Rotation=(X=0,Y=0,Z=0,W=1),Translation=(X=0,Y=0,Z=0),Scale3D=(X=1,Y=1,Z=1))
   )
   Prepare Control Shape = RigVMCollapseNode(
    Control Shape=(bVisible=True,Name="Default",Color=(R=1,G=0,B=0,A=1),Transform=(Rotation=(X=0,Y=0,Z=0,W=1),Translation=(X=0,Y=0,Z=0),Scale3D=(X=1,Y=1,Z=1)))
    Scale Factor=1
    Scale Premultiplier=1
   )
   Set Sec Control Visibility = RigVMCollapseNode(ExecuteContext=())
   Solve Ball Aim Null = RigVMCollapseNode(Backwards Solve=False)
   Solve Hock Aim Null = RigVMCollapseNode(Ankle=(Type=None,Name="None"), Base=(Type=Bone,Name="None"), End=(Type=Bone,Name="None"), Execute=(), Initial=False)
   Solve Knee PV Null = RigVMCollapseNode(End=(Type=Bone,Name="None"), Thigh=(Type=Bone,Name="None"))
   Try Setup Connection = RigVMCollapseNode(Child=(Type=Connector,Name="Knee"), ExecuteContext=(), Parent=(Type=Connector,Name="Thigh"))
  ]}
 RigVMFunction_Sequence_1_ContainedGraph {nodes [
   n1 = RigVMFunctionEntryNode()
   RigVMFunction_Sequence = &RigVMFunction_Sequence(ExecuteContext=@n1.ExecuteContext)
   n2 = &RigVMFunction_Sequence(ExecuteContext=@RigVMFunction_Sequence.B)
   n3 = &RigVMFunction_Sequence(ExecuteContext=@n2.B)
   n4 = &RigVMFunction_Sequence(ExecuteContext=@n3.B)
   n5 = &RigVMFunction_Sequence(ExecuteContext=@n4.B)
   RigVMFunctionReturnNode(A=@RigVMFunction_Sequence.A, B=@n2.A, C=@n3.A, D=@n4.A, E=@n5.A, F=@n5.B)
  ]}
 RigVMFunction_Sequence_1_ContainedGraph {nodes [
   n1 = RigVMFunctionEntryNode()
   RigVMFunction_Sequence = &RigVMFunction_Sequence(ExecuteContext=@n1.ExecuteContext)
   n2 = &RigVMFunction_Sequence(ExecuteContext=@RigVMFunction_Sequence.B)
   n3 = &RigVMFunction_Sequence(ExecuteContext=@n2.B)
   RigVMFunctionReturnNode(A=@RigVMFunction_Sequence.A, B=@n2.A, C=@n3.A, D=@n3.B)
  ]}
 RigVMFunction_Sequence_1_ContainedGraph {nodes [
   n1 = RigVMFunctionEntryNode()
   RigVMFunction_Sequence = &RigVMFunction_Sequence(ExecuteContext=@n1.ExecuteContext)
   n2 = &RigVMFunction_Sequence(ExecuteContext=@RigVMFunction_Sequence.B)
   n3 = &RigVMFunction_Sequence(ExecuteContext=@n2.B)
   n4 = &RigVMFunction_Sequence(ExecuteContext=@n3.B)
   n5 = &RigVMFunction_Sequence(ExecuteContext=@n4.B)
   RigVMFunctionReturnNode(A=@RigVMFunction_Sequence.A, B=@n2.A, C=@n3.A, D=@n4.A, E=@n5.A, F=@n5.B)
  ]}
 RigVMFunction_Sequence_1_ContainedGraph {nodes [
   n1 = RigVMFunctionEntryNode()
   RigVMFunction_Sequence = &RigVMFunction_Sequence(ExecuteContext=@n1.ExecuteContext)
   n2 = &RigVMFunction_Sequence(ExecuteContext=@RigVMFunction_Sequence.B)
   n3 = &RigVMFunction_Sequence(ExecuteContext=@n2.B)
   RigVMFunctionReturnNode(A=@RigVMFunction_Sequence.A, B=@n2.A, C=@n3.A, D=@n3.B)
  ]}
 RigVMFunction_Sequence_1_ContainedGraph {nodes [
   n1 = RigVMFunctionEntryNode()
   RigVMFunction_Sequence = &RigVMFunction_Sequence(ExecuteContext=@n1.ExecuteContext)
   n2 = &RigVMFunction_Sequence(ExecuteContext=@RigVMFunction_Sequence.B)
   n3 = &RigVMFunction_Sequence(ExecuteContext=@n2.B)
   n4 = &RigVMFunction_Sequence(ExecuteContext=@n3.B)
   RigVMFunctionReturnNode(A=@RigVMFunction_Sequence.A, B=@n2.A, C=@n3.A, D=@n4.A, E=@n4.B)
  ]}
 RigVMModel {nodes [
   RigUnit_BeginExecution = &RigUnit_BeginExecution()
   RigVMFunction_Sequence = RigVMAggregateNode(ExecuteContext=@RigUnit_BeginExecution.ExecutePin)
   Limit Transform Length = RigVMFunctionReferenceNode(
    Transform=&RigUnit_GetTransform(bInitial=False, Item=RigVMVariableNode() {variable=Ball Point Null}.Value).Transform
    Root=&RigUnit_GetTransform(bInitial=False, Item=RigVMVariableNode() {variable=Hip Control}.Value).Transform
    Length=RigVMVariableNode() {variable=Hock Chain Length}.Value
   )
   &RigUnit_SetTransform(
    Weight=1
    bInitial=False
    bPropagateToChildren=true
    Value=@"Limit Transform Length".Result
    Item=RigVMVariableNode() {variable=Ball Connector}.Value
    ExecutePin=@RigVMFunction_Sequence.C
   )
   n1 = RigVMVariableNode() {variable=Hip Connector}
   n2 = RigVMVariableNode() {variable=Hip Control}
   > &RigUnit_SetTransform(
    Weight=1
    bInitial=False
    bPropagateToChildren=false
    Value=&RigUnit_ProjectTransformToNewParent(bChildInitial=True, bNewParentInitial=False, bOldParentInitial=True, Child=@n1.Value, NewParent=@n2.Value, OldParent=@n2.Value).Transform
    Item=@n1.Value
   )
   > /Script/ControlRig.RigUnit_TwoBoneIKSimplePerItem(
    DebugSettings=(bEnabled=False,Scale=10,WorldOffset=(Rotation=(X=0,Y=0,Z=0,W=1),Translation=(X=0,Y=0,Z=0),Scale3D=(X=1,Y=1,Z=1)))
    ItemALength=0
    ItemBLength=0
    PoleVector=(X=0,Y=0,Z=0)
    PoleVectorKind=Location
    SecondaryAxisWeight=1
    StretchMaximumRatio=1.25
    StretchStartRatio=0.75
    Weight=1
    bEnableStretch=False
    bPropagateToChildren=false
    PoleVectorSpace=RigVMVariableNode() {variable=Knee PV Control}.Value
    ItemA=RigVMVariableNode() {variable=Hip Connector}.Value
    ItemB=RigVMVariableNode() {variable=Knee Connector}.Value
    EffectorItem=RigVMVariableNode() {variable=Ankle Connector}.Value
    Effector=&RigUnit_GetTransform(bInitial=False, Item=RigVMVariableNode() {variable=Ankle Roll Bottom Null}.Value).Transform
    PrimaryAxis=RigVMVariableNode() {variable=IK Primary Axis}.Value
    SecondaryAxis=RigVMVariableNode() {variable=IK Secondary Axis}.Value
   ) {is=struct RigVMUnitNode}
   Prepare Control Shape_0 = RigVMFunctionReferenceNode(
    Scale Premultiplier=1
    Control Shape=RigVMVariableNode() {variable=Ankle Roll Shape Settings}.Value
    Scale Factor=RigVMVariableNode() {variable=Control Scale}.Value
   )
   n3 = &RigUnit_GetTransform(bInitial=False, Item=RigVMVariableNode() {variable=Ankle Connector}.Value)
   n4 = &RigUnit_AimBoneMath(
    DebugSettings=(bEnabled=False,Scale=10,WorldOffset=(Rotation=(X=0,Y=0,Z=0,W=1),Translation=(X=0,Y=0,Z=0),Scale3D=(X=1,Y=1,Z=1)))
    Primary=(Weight=1,Axis=(X=1,Y=0,Z=0),Target=(X=0,Y=0,Z=0),Kind=Location,Space=(Type=Bone,Name="None"))
    Secondary=(Weight=1,Axis=(X=0,Y=0,Z=1),Target=(X=0,Y=0,Z=0),Kind=Direction,Space=(Type=Bone,Name="None"))
    Weight=1
    InputTransform=@n3.Transform
    Primary.Axis=RigVMVariableNode() {variable=IK Primary Axis}.Value
    Primary.Space=RigVMVariableNode() {variable=Ball Connector}.Value
    Secondary.Axis=RigVMVariableNode() {variable=IK Secondary Axis}.Value
    Secondary.Space=RigVMVariableNode() {variable=Knee PV Control}.Value
   )
   RigVMFunction_MathQuaternionDot = /Script/RigVM.RigVMFunction_MathQuaternionDot(B=@n4.Result.Rotation, A=@n3.Transform.Rotation) {is=struct RigVMUnitNode}
   Remap = &RigVMFunction_MathDoubleRemap(SourceMaximum=1, SourceMinimum=0.9962, TargetMaximum=0, TargetMinimum=1, bClamp=False, Value=@RigVMFunction_MathQuaternionDot.Result)
   > /Script/ControlRig.RigUnit_SetControlColor(
    Color=&RigVMFunction_MathColorLerp(
     B=(R=1,G=1,B=1,A=1)
     A=@"Prepare Control Shape_0".Result Control Shape.Color
     B.R=@"Prepare Control Shape_0".Result Control Shape.Color.B
     B.B=@"Prepare Control Shape_0".Result Control Shape.Color.G
     B.G=@"Prepare Control Shape_0".Result Control Shape.Color.R
     T=@Remap.Result
    ) {template="Interpolate::Execute(in A,in B,in T,out Result)"}.Result
    Control=/Script/ControlRig.RigUnit_ItemToName(Value=RigVMVariableNode() {variable=Ankle Roll Control}.Value) {is=struct RigVMUnitNode, template="Cast::Execute(in Value,out Result)"}.Result
   ) {is=struct RigVMUnitNode, template="SetControlColor::Execute(in Control,in Color)"}
   > RigUnit_SetTransform_5 = &RigUnit_SetRotation(
    Space=GlobalSpace
    Weight=1
    bInitial=False
    bPropagateToChildren=false
    Item=RigVMVariableNode() {variable=Ankle Connector}.Value
    Value=@n4.Result.Rotation
   )
   Solve Knee PV Null = RigVMFunctionReferenceNode(Thigh=RigVMVariableNode() {variable=Hip Control}.Value, End=RigVMVariableNode() {variable=IK Foot Control}.Value)
   n5 = &RigUnit_SetTransform(
    Value=(Rotation=(X=0,Y=0,Z=0,W=1),Translation=(X=0,Y=0,Z=0),Scale3D=(X=1,Y=1,Z=1))
    Weight=1
    bInitial=False
    bPropagateToChildren=true
    Value.Translation=@"Solve Knee PV Null".Result
    Value.Rotation=@"Solve Knee PV Null".Rotation
    Item=RigVMVariableNode() {variable=Knee PV Null}.Value
    ExecutePin=@RigVMFunction_Sequence.B
   )
   Solve Hock Aim Null = RigVMFunctionReferenceNode(
    Initial=False
    Execute=@n5.ExecutePin
    Base=RigVMVariableNode() {variable=Hip Control}.Value
    Ankle=RigVMVariableNode() {variable=Ankle Connector}.Value
    End=RigVMVariableNode() {variable=IK Foot Control}.Value
   )
   Set Transform_4 = &RigUnit_SetRotation(
    Space=GlobalSpace
    Weight=1
    bInitial=False
    bPropagateToChildren=true
    ExecutePin=@"Solve Hock Aim Null".Execute
    Value=@"Solve Hock Aim Null".Result.Rotation
    Item=RigVMVariableNode() {variable=Ankle Roll Null}.Value
   )
   Set Sec Control Visibility = RigVMFunctionReferenceNode(ExecuteContext=@RigVMFunction_Sequence.F)
   RigVMFunction_ControlFlowBranch = &RigVMFunction_ControlFlowBranch(Condition=RigVMVariableNode() {variable=Is Scapula Connected}.Value, ExecuteContext=@RigVMFunction_Sequence.A)
   n6 = RigVMVariableNode() {variable=Scapula Control}
   n7 = RigVMVariableNode() {variable=Scapula Connector}
   &RigUnit_SetTransform(
    Weight=1
    bInitial=False
    Value=&RigUnit_ProjectTransformToNewParent(bChildInitial=True, bNewParentInitial=False, bOldParentInitial=True, NewParent=@n6.Value, OldParent=@n6.Value, Child=@n7.Value).Transform
    ExecutePin=@RigVMFunction_ControlFlowBranch.True
    Item=@n7.Value
   )
   n8 = RigVMVariableNode() {variable=Ball Connector}
   n9 = &RigUnit_AimBoneMath(
    DebugSettings=(bEnabled=False,Scale=10,WorldOffset=(Rotation=(X=0,Y=0,Z=0,W=1),Translation=(X=0,Y=0,Z=0),Scale3D=(X=1,Y=1,Z=1)))
    Primary=(Weight=1,Axis=(X=1,Y=0,Z=0),Target=(X=0,Y=0,Z=0),Kind=Location,Space=(Type=Bone,Name="None"))
    Secondary=(Weight=1,Axis=(X=0,Y=0,Z=1),Target=(X=0,Y=0,Z=1),Kind=Direction,Space=(Type=Bone,Name="None"))
    Weight=1
    InputTransform=&RigUnit_GetTransform(bInitial=False, Item=@n8.Value).Transform
    Primary.Space=RigVMVariableNode() {variable=Ball FK Tip Point Null}.Value
    Secondary.Space=RigVMVariableNode() {variable=Ball FK Tip Point Null}.Value
    Primary.Axis=RigVMVariableNode() {variable=IK Primary Axis}.Value
   )
   RigUnit_SetTransform_2 = &RigUnit_SetRotation(Space=GlobalSpace, Weight=1, bInitial=False, Value=@n9.Result.Rotation, Item=@n8.Value, ExecutePin=@RigVMFunction_Sequence.D)
   n10 = RigVMVariableNode() {variable=Toe Connector}
   n11 = &RigUnit_AimBoneMath(
    DebugSettings=(bEnabled=False,Scale=10,WorldOffset=(Rotation=(X=0,Y=0,Z=0,W=1),Translation=(X=0,Y=0,Z=0),Scale3D=(X=1,Y=1,Z=1)))
    Primary=(Weight=1,Axis=(X=1,Y=0,Z=0),Target=(X=0,Y=0,Z=0),Kind=Location,Space=(Type=Bone,Name="None"))
    Secondary=(Weight=1,Axis=(X=0,Y=0,Z=1),Target=(X=0,Y=0,Z=1),Kind=Direction,Space=(Type=Bone,Name="None"))
    Weight=1
    InputTransform=&RigUnit_GetTransform(bInitial=False, Item=@n10.Value).Transform
    Primary.Space=RigVMVariableNode() {variable=Toe FK Tip Point Null}.Value
    Secondary.Space=RigVMVariableNode() {variable=Toe FK Control}.Value
    Primary.Axis=RigVMVariableNode() {variable=IK Primary Axis}.Value
   )
   > RigUnit_SetTransform_1 = &RigUnit_SetRotation(Space=GlobalSpace, Weight=1, bInitial=False, Value=@n11.Result.Rotation, Item=@n10.Value)
   n12 = RigVMVariableNode() {variable=Knee PV Control}
   RigUnit_ItemToName = &RigUnit_ItemToName(Value=@n12.Value) {template="Cast::Execute(in Value,out Result)"}
   GetControlColor = /Script/ControlRig.RigUnit_GetControlColor(Control=@RigUnit_ItemToName.Result) {is=struct RigVMUnitNode, template="GetControlColor::Execute(in Control,out Color)"}
   Interpolate = &RigVMFunction_MathColorLerp(B=(R=0,G=0,B=0,A=1), T=0.5, A=@GetControlColor.Color) {template="Interpolate::Execute(in A,in B,in T,out Result)"}
   RigVMFunction_DebugLineNoSpace = /Script/RigVM.RigVMFunction_DebugLineNoSpace(
    DebugDrawSettings=(DepthPriority=SDPG_Foreground,Lifetime=-1)
    Thickness=0.05
    WorldOffset=(Rotation=(X=0,Y=0,Z=0,W=1),Translation=(X=0,Y=0,Z=0),Scale3D=(X=1,Y=1,Z=1))
    bEnabled=true
    Color=@Interpolate.Result
    A=&RigUnit_GetTransform(bInitial=False, Item=RigVMVariableNode() {variable=Knee Connector}.Value).Transform.Translation
    B=&RigUnit_GetTransform(bInitial=False, Item=@n12.Value).Transform.Translation
    ExecutePin=@RigVMFunction_Sequence.E
   ) {is=struct RigVMUnitNode}
   n13 = &RigUnit_AimBoneMath(
    DebugSettings=(bEnabled=False,Scale=10,WorldOffset=(Rotation=(X=0,Y=0,Z=0,W=1),Translation=(X=0,Y=0,Z=0),Scale3D=(X=1,Y=1,Z=1)))
    Primary=(Weight=1,Axis=(X=-1,Y=0,Z=0),Target=(X=0,Y=0,Z=0),Kind=Location,Space=(Type=Bone,Name="None"))
    Secondary=(Weight=1,Axis=(X=0,Y=1,Z=0),Target=(X=0,Y=0,Z=0),Kind=Direction,Space=(Type=Bone,Name="None"))
    Weight=1
    InputTransform=&RigUnit_GetTransform(bInitial=False, Item=RigVMVariableNode() {variable=Hip Bottom Null}.Value).Transform
    Primary.Space=RigVMVariableNode() {variable=Ankle Roll Bottom Null}.Value
    Primary.Axis=RigVMVariableNode() {variable=IK Primary Axis}.Value
    Secondary.Axis=RigVMVariableNode() {variable=IK Secondary Axis}.Value
   )
   RigUnit_SetTransform_3 = &RigUnit_SetRotation(
    ExecutePin=()
    Space=GlobalSpace
    Weight=1
    bInitial=False
    Value=@n13.Result.Rotation
    Item=RigVMVariableNode() {variable=Hip Bottom Null}.Value
   )
   Solve Knee PV Null_1 = RigVMFunctionReferenceNode(Thigh=RigVMVariableNode() {variable=Hip Control}.Value, End=RigVMVariableNode() {variable=IK Foot Control}.Value)
   n14 = RigVMVariableNode() {variable=Knee PV Null}
   > &RigUnit_SetTransform(
    Value=(Rotation=(X=0,Y=0,Z=0,W=1),Translation=(X=0,Y=0,Z=0),Scale3D=(X=1,Y=1,Z=1))
    Weight=1
    bInitial=False
    Value.Translation=@"Solve Knee PV Null_1".Result
    Value.Rotation=@"Solve Knee PV Null_1".Rotation
    Item=@n14.Value
   )
   n15 = RigVMVariableNode() {variable=Hip Control}
   > &RigUnit_SetTransform(
    Weight=1
    bInitial=False
    Value=&RigUnit_ProjectTransformToNewParent(bChildInitial=false, bNewParentInitial=False, bOldParentInitial=True, NewParent=@n15.Value, OldParent=@n15.Value, Child=@n14.Value).Transform
    Item=@n14.Value
   )
   n16 = RigVMVariableNode() {variable=Knee Connector}
   n17 = RigVMVariableNode() {variable=Knee Rotator Control}
   RigUnit_SetTransform = &RigUnit_SetTransform(
    ExecutePin=()
    Weight=1
    bInitial=False
    bPropagateToChildren=false
    Value=&RigUnit_ProjectTransformToNewParent(bChildInitial=true, bNewParentInitial=False, bOldParentInitial=True, Child=@n16.Value, NewParent=@n17.Value, OldParent=@n17.Value).Transform
    Item=@n16.Value
   )
   Solve Ball Aim Null = RigVMFunctionReferenceNode(Backwards Solve=False)
   Set Transform = &RigUnit_SetRotation(
    ExecutePin=()
    Space=GlobalSpace
    Weight=1
    bInitial=False
    bPropagateToChildren=true
    Value=@"Solve Ball Aim Null".Result
    Item=RigVMVariableNode() {variable=Toe Roll Null}.Value
   )
   RigUnit_GetTransform = &RigUnit_GetTransform(bInitial=False, Item=RigVMVariableNode() {variable=Toe Roll Null}.Value)
   Limit Transform Length_1 = RigVMFunctionReferenceNode(
    Transform=@RigUnit_GetTransform.Transform
    Root=&RigUnit_GetTransform(bInitial=False, Item=RigVMVariableNode() {variable=Hip Control}.Value).Transform
    Length=RigVMVariableNode() {variable=Chain Length}.Value
   )
   &RigUnit_SetTransform(
    ExecutePin=()
    Weight=1
    bInitial=False
    bPropagateToChildren=true
    Value=@"Limit Transform Length_1".Result
    Item=RigVMVariableNode() {variable=Toe Connector}.Value
   )
   GetAnimationChannelFromItem = &RigUnit_GetBoolAnimationChannelFromItem(bInitial=False, Item=RigVMVariableNode() {variable=Knee Rotator Vis}.Value) {template="GetAnimationChannelFromItem::Execute(out Value,in Item,in bInitial)"}
   SetControlVisibility = &RigUnit_SetControlVisibility(ExecutePin=(), bVisible=@GetAnimationChannelFromItem.Value, Item=@n17.Value) {template="SetControlVisibility::Execute(in Item,in Pattern,in bVisible)"}
   RigVMVariableNode() {variable=Knee Connector}
   RigVMVariableNode() {variable=Knee PV Null}
  ]}
 RigVMModel Connection Event {nodes [
   n1 = /Script/ControlRig.RigUnit_ConnectorExecution() {is=struct RigVMUnitNode}
   Try Setup Connection = RigVMFunctionReferenceNode(Child=(Type=Connector,Name="Mid"), Parent=(Type=Connector,Name="Start"), ExecuteContext=@n1.ExecuteContext)
   > Try Setup Connection_1 = RigVMFunctionReferenceNode(Child=(Type=Connector,Name="End"), Parent=(Type=Connector,Name="Mid"))
   > Try Setup Connection_2 = RigVMFunctionReferenceNode(Child=(Type=Connector,Name="Ball"), Parent=(Type=Connector,Name="End"))
   > Try Setup Connection_4 = RigVMFunctionReferenceNode(Child=(Type=Connector,Name="Toe"), Parent=(Type=Connector,Name="Ball"))
   > Try Setup Connection_3 = RigVMFunctionReferenceNode(Child=(Type=Connector,Name="Toe Tip"), Parent=(Type=Connector,Name="Toe"))
  ]}
 RigVMModel Rig {nodes [
   n1 = /Script/ControlRig.RigUnit_InverseExecution() {is=struct RigVMUnitNode}
   RigVMFunction_Sequence = RigVMAggregateNode(ExecuteContext=@n1.ExecutePin)
   n2 = RigVMVariableNode() {variable=IK Foot Control}
   n3 = RigVMVariableNode() {variable=Toe Connector}
   RigUnit_SetTransform = &RigUnit_SetTransform(
    Weight=1
    bInitial=False
    Value=&RigUnit_ProjectTransformToNewParent(bChildInitial=True, bNewParentInitial=False, bOldParentInitial=True, Child=@n2.Value, NewParent=@n3.Value, OldParent=@n3.Value).Transform
    Item=@n2.Value
    ExecutePin=@RigVMFunction_Sequence.A
   )
   n4 = RigVMVariableNode() {variable=Ball Connector}
   n5 = RigVMVariableNode() {variable=Toe Roll Control}
   > n6 = &RigUnit_SetTransform(
    Weight=1
    bInitial=False
    Value=&RigUnit_ProjectTransformToNewParent(bChildInitial=True, bNewParentInitial=False, bOldParentInitial=True, NewParent=@n4.Value, OldParent=@n4.Value, Child=@n5.Value).Transform
    Item=@n5.Value
   )
   RigVMFunction_ControlFlowBranch = &RigVMFunction_ControlFlowBranch(ExecuteContext=@n6.ExecutePin, Condition=RigVMVariableNode() {variable=Is Scapula Connected}.Value)
   n7 = RigVMVariableNode() {variable=Hip Control}
   n8 = RigVMVariableNode() {variable=Hip Connector}
   Set Transform_1 = /Script/ControlRig.RigUnit_SetTranslation(
    Space=GlobalSpace
    Weight=1
    bInitial=False
    bPropagateToChildren=True
    Value=&RigUnit_ProjectTransformToNewParent(bChildInitial=True, bNewParentInitial=False, bOldParentInitial=True, Child=@n7.Value, NewParent=@n8.Value, OldParent=@n8.Value).Transform.Translation
    ExecutePin=@RigVMFunction_ControlFlowBranch.False
    Item=@n7.Value
   ) {is=struct RigVMUnitNode, template="Set Transform::Execute(in Item,in Space,in bInitial,in Value,in Weight,in bPropagateToChildren)"}
   n9 = RigVMVariableNode() {variable=Scapula Control}
   n10 = RigVMVariableNode() {variable=Scapula Connector}
   &RigUnit_SetTransform(
    Weight=1
    bInitial=False
    bPropagateToChildren=false
    Value=&RigUnit_ProjectTransformToNewParent(bChildInitial=True, bNewParentInitial=False, bOldParentInitial=True, Child=@n9.Value, NewParent=@n10.Value, OldParent=@n10.Value).Transform
    ExecutePin=@RigVMFunction_ControlFlowBranch.True
    Item=@n9.Value
   )
   Solve Knee PV Null = RigVMFunctionReferenceNode(Thigh=RigVMVariableNode() {variable=Hip Connector}.Value, End=RigVMVariableNode() {variable=Ball Connector}.Value)
   &RigUnit_SetTransform(
    Value=(Rotation=(X=0,Y=0,Z=0,W=1),Translation=(X=0,Y=0,Z=0),Scale3D=(X=1,Y=1,Z=1))
    Weight=1
    bInitial=False
    Value.Translation=@"Solve Knee PV Null".Result
    Value.Rotation=@"Solve Knee PV Null".Rotation
    Item=RigVMVariableNode() {variable=Knee PV Null}.Value
    ExecutePin=@RigVMFunction_Sequence.B
   )
   Solve Ball Aim Null = RigVMFunctionReferenceNode(Backwards Solve=false)
   > RigUnit_SetTransform_2 = &RigUnit_SetRotation(Space=GlobalSpace, Weight=1, bInitial=False, Value=@"Solve Ball Aim Null".Result, Item=RigVMVariableNode() {variable=Toe Roll Null}.Value)
   Solve Hock Aim Null = RigVMFunctionReferenceNode(
    Initial=false
    Execute=@RigUnit_SetTransform_2.ExecutePin
    Ankle=RigVMVariableNode() {variable=Ankle Connector}.Value
    End=RigVMVariableNode() {variable=Ball Connector}.Value
    Base=RigVMVariableNode() {variable=Hip Connector}.Value
   )
   RigUnit_SetTransform_1 = &RigUnit_SetRotation(
    Space=GlobalSpace
    Weight=1
    bInitial=False
    ExecutePin=@"Solve Hock Aim Null".Execute
    Value=@"Solve Hock Aim Null".Result.Rotation
    Item=RigVMVariableNode() {variable=Ankle Roll Null}.Value
   )
   n11 = RigVMVariableNode() {variable=Ankle Connector}
   n12 = RigVMVariableNode() {variable=Ankle Roll Control}
   > RigUnit_SetTransform_3 = &RigUnit_SetRotation(
    Space=GlobalSpace
    Weight=1
    bInitial=False
    Value=&RigUnit_ProjectTransformToNewParent(bChildInitial=True, bNewParentInitial=false, bOldParentInitial=True, NewParent=@n11.Value, OldParent=@n11.Value, Child=@n12.Value).Transform.Rotation
    Item=@n12.Value
   )
   Compute Pole Vector Location v02 = RigVMFunctionReferenceNode(
    Bones=((Type=None,Name="None"),(Type=None,Name="None"),(Type=None,Name="None"))
    Secondary Axis=(X=0,Y=0,Z=0)
    Bones.0=RigVMVariableNode() {variable=Hip Connector}.Value
    Bones.1=RigVMVariableNode() {variable=Knee Connector}.Value
    Bones.2=RigVMVariableNode() {variable=Ankle Connector}.Value
    PV Offset=RigVMVariableNode() {variable=PV Distance Scale}.Value
   )
   RigUnit_GetTransform = &RigUnit_GetTransform(bInitial=False, Item=RigVMVariableNode() {variable=Knee Connector}.Value)
   RigVMFunction_MathVectorDistance = &RigVMFunction_MathVectorDistance(B=@"Compute Pole Vector Location v02".Pole Vector Location, A=@RigUnit_GetTransform.Transform.Translation)
   Less = /Script/RigVM.RigVMFunction_MathDoubleLess(B=0.01, A=@RigVMFunction_MathVectorDistance.Result) {is=struct RigVMUnitNode, template="Less::Execute(in A,in B,out Result)"}
   n13 = RigVMVariableNode() {variable=Knee PV Control}
   &RigUnit_SetTransform(
    Value=(Rotation=(X=0,Y=0,Z=0,W=1),Translation=(X=0,Y=0,Z=0),Scale3D=(X=1,Y=1,Z=1))
    Weight=1
    bInitial=False
    Value.Translation=RigVMDispatchNode(
     Condition=@Less.Result
     True=&RigUnit_GetTransform(bInitial=true, Item=@n13.Value).Transform.Translation
     False=@"Compute Pole Vector Location v02".Pole Vector Location
    ) {template="DISPATCH_RigVMDispatch_If(in Condition,in True,in False,out Result)"}.Result
    Item=RigVMVariableNode() {variable=Knee PV Control}.Value
    ExecutePin=@RigVMFunction_Sequence.C
   )
   > &RigUnit_SetTransform(
    Weight=1
    bInitial=False
    Value=&RigUnit_GetTransform(bInitial=False, Item=RigVMVariableNode() {variable=Toe Connector}.Value).Transform
    Item=RigVMVariableNode() {variable=Toe FK Control}.Value
   )
   &RigUnit_SetTransform(
    ExecutePin=()
    Item=(Type=Bone,Name="None")
    Weight=1
    bInitial=False
    Value=&RigUnit_GetTransform(bInitial=False, Item=RigVMVariableNode() {variable=Toe Tip Connector}.Value).Transform
   )
   RigVMVariableNode() {variable=Knee Connector}
   RigVMVariableNode() {variable=Hip Connector}
  ]}
]
hierarchy [
 Start {type=Connector, loc=<0 0 0>, rot=<0 0 0 1>}
 Root {type=Socket, loc=<0 0 0>, rot=<0 0 0 1>}
 Mid {type=Connector, loc=<0 0 0>, rot=<0 0 0 1>}
 End {type=Connector, loc=<0 0 0>, rot=<0 0 0 1>}
 Ball {type=Connector, loc=<0 0 0>, rot=<0 0 0 1>}
 Toe {type=Connector, loc=<0 0 0>, rot=<0 0 0 1>}
 Toe Tip {type=Connector, loc=<0 0 0>, rot=<0 0 0 1>}
 Parent {type=Connector, loc=<0 0 0>, rot=<0 0 0 1>}
 IK Parent {type=Connector, loc=<0 0 0>, rot=<0 0 0 1>}
 IK Spaces {type=Connector, loc=<0 0 0>, rot=<0 0 0 1>}
 Scapula {type=Connector, loc=<0 0 0>, rot=<0 0 0 1>}
]
variables [
 Ankle Connector {type=FRigElementKey, default=(Type=None,Name="")}
 Ankle Driver Null {type=FRigElementKey, default=(Type=Bone,Name="")}
 Ankle Initial Interp Target {type=double, default=0.05}
 Ankle Roll Bottom Null {type=FRigElementKey, default=(Type=Bone,Name="")}
 Ankle Roll Control {type=FRigElementKey, default=(Type=None,Name="")}
 Ankle Roll Null {type=FRigElementKey, default=(Type=Bone,Name="")}
 Ankle Roll Shape Settings {
  type=FRigUnit_HierarchyAddControl_ShapeSettings
  default=(bVisible=True,Name="Circle_Thick",Color=(R=0,G=0.402,B=0.7011,A=1),Transform=(Rotation=(X=0,Y=0,Z=0,W=1),Translation=(X=1.5,Y=0,Z=0),Scale3D=(X=0.5,Y=0.5,Z=0.5)))
 }
 Ball Aim Zero Ref Null {type=FRigElementKey, default=(Type=None,Name="")}
 Ball Connector {type=FRigElementKey, default=(Type=None,Name="")}
 Ball FK Control {type=FRigElementKey, default=(Type=None,Name="")}
 Ball FK Shape Settings {
  type=FRigUnit_HierarchyAddControl_ShapeSettings
  default=(bVisible=True,Name="Circle_Thick",Color=(R=1,G=1,B=1,A=1),Transform=(Rotation=(X=0,Y=-0.7071,Z=0,W=0.7071),Translation=(X=0,Y=0,Z=0),Scale3D=(X=1,Y=1,Z=1)))
 }
 Ball FK Tip Point Null {type=FRigElementKey, default=(Type=None,Name="")}
 Ball Point Null {type=FRigElementKey, default=(Type=Bone,Name="")}
 Calculated PV Control Location {type=FVector, default=(X=0,Y=0,Z=0)}
 Chain Length {type=float, default=0}
 Control Parent Connector {type=FRigElementKey, default=(Type=None,Name="")}
 Control Scale {type=double, default=1}
 Heel Pivot Control {type=FRigElementKey, default=(Type=None,Name="")}
 Heel Pivot Shape Settings {
  type=FRigUnit_HierarchyAddControl_ShapeSettings
  default=(bVisible=True,Name="Sphere_Solid",Color=(R=1,G=1,B=1,A=1),Transform=(Rotation=(X=0,Y=0,Z=0,W=1),Translation=(X=0,Y=-1,Z=1.5),Scale3D=(X=1,Y=1,Z=1)))
 }
 Heel Roll Offset Transform {type=FTransform, default=(Rotation=(X=0,Y=0,Z=0,W=1),Translation=(X=0,Y=-2,Z=-2),Scale3D=(X=1,Y=1,Z=1))}
 Hip Bottom Null {type=FRigElementKey, default=(Type=None,Name="")}
 Hip Connector {type=FRigElementKey, default=(Type=None,Name="")}
 Hip Control {type=FRigElementKey, default=(Type=None,Name="")}
 Hip Shape Settings {
  type=FRigUnit_HierarchyAddControl_ShapeSettings
  default=(bVisible=True,Name="Box_Thick",Color=(R=1,G=1,B=1,A=1),Transform=(Rotation=(X=0,Y=0,Z=0,W=1),Translation=(X=0,Y=0,Z=0),Scale3D=(X=1,Y=1,Z=1)))
 }
 Hock Aim Null Transform {type=FTransform, default=(Rotation=(X=0,Y=0,Z=0,W=1),Translation=(X=0,Y=0,Z=0),Scale3D=(X=1,Y=1,Z=1))}
 Hock Angle Sweep {type=double, default=60}
 Hock Chain Length {type=float, default=0}
 IK Foot Control {type=FRigElementKey, default=(Type=None,Name="")}
 IK Primary Axis {type=FVector, default=(X=1,Y=0,Z=0)}
 IK Secondary Axis {type=FVector, default=(X=0,Y=1,Z=0)}
 IK Shape Settings {
  type=FRigUnit_HierarchyAddControl_ShapeSettings
  default=(bVisible=True,Name="RoundedTrapeze_Thick",Color=(R=1,G=1,B=1,A=1),Transform=(Rotation=(X=0,Y=0,Z=1,W=0),Translation=(X=0.3,Y=2,Z=-2),Scale3D=(X=1,Y=1,Z=1)))
 }
 IK Spaces {type="TArray<FRigElementKey>", default=(())}
 Initial Ball Length {type=float, default=0}
 Initial Distance Hip to IK Target {type=float, default=0}
 Initial Knee PV Null Orientation {type=FQuat, default=(X=0,Y=0,Z=0,W=1)}
 Initial PV Null Location {type=FVector, default=(X=0,Y=0,Z=0)}
 Initial Thigh Length {type=float, default=0}
 Initial Thigh to Ankle Length {type=double, default=0}
 Invert Heel Secondary Axis {type=bool, default=False}
 Is Scapula Connected {type=bool, default=False}
 Is Toe Tip Connected {type=bool, default=False}
 Knee Connector {type=FRigElementKey, default=(Type=None,Name="")}
 Knee PV Control {type=FRigElementKey, default=(Type=None,Name="")}
 Knee PV Null {type=FRigElementKey, default=(Type=None,Name="")}
 Knee Rotate Shape Settings {
  type=FRigUnit_HierarchyAddControl_ShapeSettings
  default=(bVisible=True,Name="Circle_Thin",Color=(R=1,G=1,B=1,A=1),Transform=(Rotation=(X=0,Y=0,Z=0,W=1),Translation=(X=0,Y=0,Z=0),Scale3D=(X=1,Y=1,Z=1)))
 }
 Knee Rotator Control {type=FRigElementKey, default=(Type=None,Name="")}
 Knee Rotator Vis {type=FRigElementKey, default=(Type=None,Name="")}
 Local Squash Target {type=FQuat, default=(X=0,Y=0,Z=0,W=1)}
 PV Distance Scale {type=double, default=1}
 PV Null Ratio {type=float, default=0.425}
 PV Shape Settings {
  type=FRigUnit_HierarchyAddControl_ShapeSettings
  default=(bVisible=True,Name="Diamond_Solid",Color=(R=1,G=1,B=1,A=1),Transform=(Rotation=(X=0,Y=0,Z=0,W=1),Translation=(X=0,Y=0,Z=0),Scale3D=(X=1,Y=1,Z=1)))
 }
 Pelvis Connector {type=FRigElementKey, default=(Type=None,Name="")}
 Root Module Color Weight {type=double, default=1}
 Scapula Connector {type=FRigElementKey, default=(Type=None,Name="")}
 Scapula Control {type=FRigElementKey, default=(Type=Bone,Name="")}
 Scapula Shape Settings {
  type=FRigUnit_HierarchyAddControl_ShapeSettings
  default=(bVisible=True,Name="Sphere_Solid",Color=(R=1,G=1,B=1,A=1),Transform=(Rotation=(X=0,Y=0,Z=0,W=1),Translation=(X=0,Y=0,Z=-5),Scale3D=(X=1,Y=1,Z=1)))
 }
 Sec Control Vis Channel {type=FRigElementKey, default=(Type=Bone,Name="")}
 Toe Connector {type=FRigElementKey, default=(Type=None,Name="")}
 Toe FK Control {type=FRigElementKey, default=(Type=None,Name="")}
 Toe FK Shape Settings {
  type=FRigUnit_HierarchyAddControl_ShapeSettings
  default=(bVisible=True,Name="Circle_Thick",Color=(R=1,G=1,B=1,A=1),Transform=(Rotation=(X=0,Y=-0.7071,Z=0,W=0.7071),Translation=(X=0,Y=0,Z=0),Scale3D=(X=1,Y=1,Z=1)))
 }
 Toe FK Tip Point Null {type=FRigElementKey, default=(Type=None,Name="")}
 Toe Pivot Control {type=FRigElementKey, default=(Type=None,Name="")}
 Toe Pivot Shape Settings {
  type=FRigUnit_HierarchyAddControl_ShapeSettings
  default=(bVisible=True,Name="Sphere_Solid",Color=(R=1,G=1,B=1,A=1),Transform=(Rotation=(X=0,Y=0,Z=0,W=1),Translation=(X=0,Y=1,Z=1.5),Scale3D=(X=1,Y=1,Z=1)))
 }
 Toe Roll Control {type=FRigElementKey, default=(Type=None,Name="")}
 Toe Roll Null {type=FRigElementKey, default=(Type=None,Name="")}
 Toe Roll Shape Settings {
  type=FRigUnit_HierarchyAddControl_ShapeSettings
  default=(bVisible=True,Name="HalfCircle_Thick",Color=(R=0,G=0.4,B=0.7,A=1),Transform=(Rotation=(X=0,Y=-0.7071,Z=0,W=0.7071),Translation=(X=0,Y=0,Z=0),Scale3D=(X=1,Y=1,Z=1)))
 }
 Toe Roll Tip Offset Transform {type=FTransform, default=(Rotation=(X=0,Y=0,Z=0,W=1),Translation=(X=0,Y=0,Z=0.5),Scale3D=(X=1,Y=1,Z=1))}
 Toe Tip Connector {type=FRigElementKey, default=(Type=None,Name="")}
]
